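# Physics asset for the skydive coat mesh of the female medium Galactic player body, defining capsule bodies for the jacket's dynamic bones.
# Front, inner-front, side and back flaps on each side form three-segment chains with limited swing and twist, hung from angularly free master bodies on the pelvis.
# Free-angle constraints cross-link middle back segments to each other (duplicated) and to the sides.
# Pelvis and leg capsules plus dyn_floor_COL spheres provide collision shapes.
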PhysicsAsset /BRCosmetics/Characters/Player/Female/Medium/Bodies/F_MED_Galactic/Meshes/F_MED_Galactic_coat_skydive_Physics
bodies {
 calf_l [{shapeType=Capsule, length=45.6211, radius=8.3375, center=<-28.7011 -0.7036 0.0466>, rotation=<rot 0 89.533 -88.5703>}]
 calf_r [{shapeType=Capsule, length=45.6211, radius=8.3375, center=<28.7258 0.7047 1.4779>, rotation=<rot -0.3385 89.533 -88.4634>}]
 dyn_floor_COL [
  {shapeType=Sphere, radius=187.1686, shapeName=dyn_floor_COL_sphere, center=<-3.4923 -2.7141 -174.8455>}
  {shapeType=Sphere, radius=24.9585, shapeName=dyn_floor_COL_sphere, center=<0 2.4787 8.5972>}
 ]
 dyn_jacketBack_01_l [{shapeType=Capsule, length=8.903, radius=9.1456, shapeName=dyn_jacketBack_01_l_capsule, center=<9 0 8>, rotation=<rot 0 0 90>}]
 dyn_jacketBack_01_r [{shapeType=Capsule, length=8.903, radius=9.1456, shapeName=dyn_jacketBack_01_r_capsule, center=<9 0 8>, rotation=<rot 0 0 90>}]
 dyn_jacketBack_02_l [{shapeType=Capsule, length=15.8606, radius=10.0118, shapeName=dyn_jacketBack_02_l_capsule, center=<8 0 8>, rotation=<rot 0 -10.0014 89.9999>}]
 dyn_jacketBack_02_r [{shapeType=Capsule, length=15.8606, radius=10.0118, shapeName=dyn_jacketBack_02_r_capsule, center=<8 0 8>, rotation=<rot 0 10.0007 90>}]
 dyn_jacketBack_03_l [{shapeType=Capsule, length=18.7013, radius=10.7576, shapeName=dyn_jacketBack_03_l_capsule, center=<8 0 8>, rotation=<rot 0 -15.0003 89.9999>}]
 dyn_jacketBack_03_r [{shapeType=Capsule, length=18.7013, radius=10.7576, shapeName=dyn_jacketBack_03_r_capsule, center=<8 0 8>, rotation=<rot 0 15.0013 89.9999>}]
 dyn_jacketBack_mstr_l [{shapeType=Capsule, length=0.505, radius=0.505, shapeName=dyn_jacketBack_mstr_l_capsule, rotation=<rot 0 -90 0>}]
 dyn_jacketBack_mstr_r [{shapeType=Capsule, length=0.505, radius=0.505, shapeName=dyn_jacketBack_mstr_r_capsule, rotation=<rot 0 -90 0>}]
 dyn_jacketFront_01_l [{shapeType=Capsule, length=14.4835, radius=5.9529, shapeName=dyn_jacketFront_01_l_capsule, center=<10 3 6>, rotation=<rot 0 0 120.0021>}]
 dyn_jacketFront_01_r [{shapeType=Capsule, length=14.4835, radius=5.9529, shapeName=dyn_jacketFront_01_r_capsule, center=<10 -3 6>, rotation=<rot 0 0 49.9986>}]
 dyn_jacketFront_02_l [{
   shapeType=Capsule
   length=16.4117
   radius=8.1356
   shapeName=dyn_jacketFront_02_l_capsule
   center=<8 3 6>
   rotation=<rot 2.4979 -4.333 119.9076>
  }]
 dyn_jacketFront_02_r [{
   shapeType=Capsule
   length=16.4117
   radius=8.1356
   shapeName=dyn_jacketFront_02_r_capsule
   center=<8 -3 6>
   rotation=<rot 3.2117 3.8343 50.1069>
  }]
 dyn_jacketFront_03_l [{
   shapeType=Capsule
   length=14.2199
   radius=9.4617
   shapeName=dyn_jacketFront_03_l_capsule
   center=<9 3 6>
   rotation=<rot 9.8471 -17.496 118.4823>
  }]
 dyn_jacketFront_03_r [{
   shapeType=Capsule
   length=13.965
   radius=9.731
   shapeName=dyn_jacketFront_03_r_capsule
   center=<9 -3 6>
   rotation=<rot 9.5765 11.6009 50.9767>
  }]
 dyn_jacketFront_inner_01_l [{shapeType=Capsule, length=6.5279, radius=6.4721, shapeName=dyn_jacketFront_inner_01_l_capsule, center=<10 6 3>, rotation=<rot 0 0 -35>}]
 dyn_jacketFront_inner_01_r [{shapeType=Capsule, length=6.5279, radius=6.4721, shapeName=dyn_jacketFront_inner_01_r_capsule, center=<10 -6 3>, rotation=<rot 0 0 35>}]
 dyn_jacketFront_inner_02_l [{shapeType=Capsule, length=16.006, radius=8.994, shapeName=dyn_jacketFront_inner_02_l_capsule, center=<10 7 3>, rotation=<rot 0 0 -39>}]
 dyn_jacketFront_inner_02_r [{shapeType=Capsule, length=16.006, radius=8.994, shapeName=dyn_jacketFront_inner_02_r_capsule, center=<10 -7 3>, rotation=<rot 0 0 35>}]
 dyn_jacketFront_inner_03_l [{shapeType=Capsule, length=15.5, radius=9.5, shapeName=dyn_jacketFront_inner_03_l_capsule, center=<10 7 3>, rotation=<rot 0 0 -39>}]
 dyn_jacketFront_inner_03_r [{shapeType=Capsule, length=15.5, radius=9.5, shapeName=dyn_jacketFront_inner_03_r_capsule, center=<10 -7 3>, rotation=<rot 0 0 35>}]
 dyn_jacketFront_inner_mstr_l [{shapeType=Capsule, length=0.505, radius=0.505, shapeName=dyn_jacketFront_inner_mstr_l_capsule, rotation=<rot 0 -90 0>}]
 dyn_jacketFront_inner_mstr_r [{shapeType=Capsule, length=0.505, radius=0.505, shapeName=dyn_jacketFront_inner_mstr_r_capsule, rotation=<rot 0 -90 0>}]
 dyn_jacketFront_mstr_l [{shapeType=Capsule, length=0.505, radius=0.505, shapeName=dyn_jacketFront_mstr_l_capsule, rotation=<rot 0 -90 0>}]
 dyn_jacketFront_mstr_r [{shapeType=Capsule, length=0.505, radius=0.505, shapeName=dyn_jacketFront_mstr_r_capsule, rotation=<rot 90 0 0>}]
 dyn_jacketSide_01_l [{shapeType=Capsule, length=8.6241, radius=8.3759, shapeName=dyn_jacketSide_01_l_capsule, center=<9 3 8>, rotation=<rot 0 0 95.0009>}]
 dyn_jacketSide_01_r [{shapeType=Capsule, length=8.6241, radius=8.3759, shapeName=dyn_jacketSide_01_r_capsule, center=<9 -3 8>, rotation=<rot 0 0 90>}]
 dyn_jacketSide_02_l [{shapeType=Capsule, length=10.9313, radius=10.0282, shapeName=dyn_jacketSide_02_l_capsule, center=<7 3 8>, rotation=<rot 0 -5.0006 95.0007>}]
 dyn_jacketSide_02_r [{shapeType=Capsule, length=10.9313, radius=10.0282, shapeName=dyn_jacketSide_02_r_capsule, center=<7 -3 8>, rotation=<rot 0 5.0006 89.9996>}]
 dyn_jacketSide_03_l [{shapeType=Capsule, length=14.6692, radius=11.4106, shapeName=dyn_jacketSide_03_l_capsule, center=<9 3 8>, rotation=<rot 0 -15.0039 95.0009>}]
 dyn_jacketSide_03_r [{shapeType=Capsule, length=14.6692, radius=11.4106, shapeName=dyn_jacketSide_03_r_capsule, center=<9 -3 8>, rotation=<rot 0 15.0044 90.0018>}]
 dyn_jacketSide_mstr_l [{shapeType=Capsule, length=0.505, radius=0.505, shapeName=dyn_jacketSide_mstr_l_capsule, rotation=<rot 0 -90 0>}]
 dyn_jacketSide_mstr_r [{shapeType=Capsule, length=0.505, radius=0.505, shapeName=dyn_jacketSide_mstr_r_capsule, rotation=<rot 0 0 90>}]
 pelvis [
  {shapeType=Capsule, length=9.0875, radius=9.6801, center=<-0.9227 -1.2812 0>}
  {shapeType=Capsule, length=13.415, radius=7.4702, shapeName=pelvis_capsule, center=<-3.9978 -3.7131 0>}
 ]
 thigh_l [{shapeType=Capsule, length=29.8476, radius=10.2576, center=<-24 1 1.1>, rotation=<rot 90 0 -1.689>}]
 thigh_r [{shapeType=Capsule, length=30.6127, radius=10.2327, center=<23.9597 1 -1.1636>, rotation=<rot 90 0 -1.6898>}]
 thigh_twist_01_r [{
   shapeType=Capsule
   length=10.1925
   radius=7.2202
   shapeName=thigh_twist_01_r_capsule
   center=<4.8925 -0.1036 3.2898>
   rotation=<rot 80.0081 2.5199 -2.4994>
  }]
}
constraints [
 {bone1Name=thigh_l, bone2Name=pelvis, swing1LimitDegrees=45, swing2LimitDegrees=45, twistLimitDegrees=45}
 {bone1Name=calf_l, bone2Name=thigh_l, swing1LimitDegrees=45, swing2LimitDegrees=45, twistLimitDegrees=45}
 {bone1Name=thigh_r, bone2Name=pelvis, swing1LimitDegrees=45, swing2LimitDegrees=45, twistLimitDegrees=45}
 {bone1Name=calf_r, bone2Name=thigh_r, swing1LimitDegrees=45, swing2LimitDegrees=45, twistLimitDegrees=45}
 {
  bone1Name=dyn_jacketFront_mstr_r
  bone2Name=pelvis
  swing1LimitDegrees=45
  swing1Motion=Free
  swing2LimitDegrees=45
  swing2Motion=Free
  twistLimitDegrees=45
  twistMotion=Free
 }
 {
  bone1Name=dyn_jacketFront_mstr_l
  bone2Name=pelvis
  swing1LimitDegrees=45
  swing1Motion=Free
  swing2LimitDegrees=45
  swing2Motion=Free
  twistLimitDegrees=45
  twistMotion=Free
 }
 {
  bone1Name=dyn_jacketSide_mstr_r
  bone2Name=pelvis
  swing1LimitDegrees=45
  swing1Motion=Free
  swing2LimitDegrees=45
  swing2Motion=Free
  twistLimitDegrees=45
  twistMotion=Free
 }
 {
  bone1Name=dyn_jacketBack_mstr_r
  bone2Name=pelvis
  swing1LimitDegrees=45
  swing1Motion=Free
  swing2LimitDegrees=45
  swing2Motion=Free
  twistLimitDegrees=45
  twistMotion=Free
 }
 {
  bone1Name=dyn_jacketBack_mstr_l
  bone2Name=pelvis
  swing1LimitDegrees=45
  swing1Motion=Free
  swing2LimitDegrees=45
  swing2Motion=Free
  twistLimitDegrees=45
  twistMotion=Free
 }
 {
  bone1Name=dyn_jacketSide_mstr_l
  bone2Name=pelvis
  swing1LimitDegrees=45
  swing1Motion=Free
  swing2LimitDegrees=45
  swing2Motion=Free
  twistLimitDegrees=45
  twistMotion=Free
 }
 {
  bone1Name=dyn_jacketFront_inner_mstr_r
  bone2Name=pelvis
  swing1LimitDegrees=45
  swing1Motion=Free
  swing2LimitDegrees=45
  swing2Motion=Free
  twistLimitDegrees=45
  twistMotion=Free
 }
 {
  bone1Name=dyn_jacketFront_inner_mstr_l
  bone2Name=pelvis
  swing1LimitDegrees=45
  swing1Motion=Free
  swing2LimitDegrees=45
  swing2Motion=Free
  twistLimitDegrees=45
  twistMotion=Free
 }
 {
  bone1Name=dyn_jacketFront_inner_01_r
  bone2Name=dyn_jacketFront_inner_mstr_r
  swing1LimitDegrees=35
  swing2LimitDegrees=10
  twistLimitDegrees=10
 }
 {
  bone1Name=dyn_jacketFront_inner_02_r
  bone2Name=dyn_jacketFront_inner_01_r
  swing1LimitDegrees=35
  swing2LimitDegrees=10
  twistLimitDegrees=10
 }
 {
  bone1Name=dyn_jacketFront_inner_03_r
  bone2Name=dyn_jacketFront_inner_02_r
  swing1LimitDegrees=35
  swing2LimitDegrees=10
  twistLimitDegrees=10
 }
 {
  bone1Name=dyn_jacketFront_inner_01_l
  bone2Name=dyn_jacketFront_inner_mstr_l
  swing1LimitDegrees=35
  swing2LimitDegrees=10
  twistLimitDegrees=10
 }
 {
  bone1Name=dyn_jacketFront_inner_02_l
  bone2Name=dyn_jacketFront_inner_01_l
  swing1LimitDegrees=35
  swing2LimitDegrees=10
  twistLimitDegrees=10
 }
 {
  bone1Name=dyn_jacketFront_inner_03_l
  bone2Name=dyn_jacketFront_inner_02_l
  swing1LimitDegrees=35
  swing2LimitDegrees=10
  twistLimitDegrees=10
 }
 {bone1Name=dyn_jacketSide_01_l, bone2Name=dyn_jacketSide_mstr_l, swing1LimitDegrees=35, swing2LimitDegrees=35, twistLimitDegrees=15}
 {bone1Name=dyn_jacketSide_02_l, bone2Name=dyn_jacketSide_01_l, swing1LimitDegrees=35, swing2LimitDegrees=35, twistLimitDegrees=15}
 {bone1Name=dyn_jacketSide_03_l, bone2Name=dyn_jacketSide_02_l, swing1LimitDegrees=35, swing2LimitDegrees=35, twistLimitDegrees=15}
 {bone1Name=dyn_jacketBack_01_l, bone2Name=dyn_jacketBack_mstr_l, swing1LimitDegrees=15, swing2LimitDegrees=35, twistLimitDegrees=5}
 {bone1Name=dyn_jacketBack_02_l, bone2Name=dyn_jacketBack_01_l, swing1LimitDegrees=15, swing2LimitDegrees=35, twistLimitDegrees=5}
 {bone1Name=dyn_jacketBack_03_l, bone2Name=dyn_jacketBack_02_l, swing1LimitDegrees=15, swing2LimitDegrees=35, twistLimitDegrees=5}
 {bone1Name=dyn_jacketBack_01_r, bone2Name=dyn_jacketBack_mstr_r, swing1LimitDegrees=15, swing2LimitDegrees=35, twistLimitDegrees=5}
 {bone1Name=dyn_jacketBack_02_r, bone2Name=dyn_jacketBack_01_r, swing1LimitDegrees=15, swing2LimitDegrees=35, twistLimitDegrees=5}
 {bone1Name=dyn_jacketBack_03_r, bone2Name=dyn_jacketBack_02_r, swing1LimitDegrees=15, swing2LimitDegrees=35, twistLimitDegrees=5}
 {bone1Name=dyn_jacketSide_01_r, bone2Name=dyn_jacketSide_mstr_r, swing1LimitDegrees=35, swing2LimitDegrees=35, twistLimitDegrees=15}
 {bone1Name=dyn_jacketSide_02_r, bone2Name=dyn_jacketSide_01_r, swing1LimitDegrees=35, swing2LimitDegrees=35, twistLimitDegrees=15}
 {bone1Name=dyn_jacketSide_03_r, bone2Name=dyn_jacketSide_02_r, swing1LimitDegrees=35, swing2LimitDegrees=35, twistLimitDegrees=15}
 {bone1Name=dyn_jacketFront_01_l, bone2Name=dyn_jacketFront_mstr_l, swing1LimitDegrees=35, swing2LimitDegrees=35, twistLimitDegrees=15}
 {bone1Name=dyn_jacketFront_02_l, bone2Name=dyn_jacketFront_01_l, swing1LimitDegrees=35, swing2LimitDegrees=35, twistLimitDegrees=15}
 {bone1Name=dyn_jacketFront_03_l, bone2Name=dyn_jacketFront_02_l, swing1LimitDegrees=35, swing2LimitDegrees=35, twistLimitDegrees=15}
 {bone1Name=dyn_jacketFront_01_r, bone2Name=dyn_jacketFront_mstr_r, swing1LimitDegrees=15, swing2LimitDegrees=35, twistLimitDegrees=15}
 {bone1Name=dyn_jacketFront_02_r, bone2Name=dyn_jacketFront_01_r, swing1LimitDegrees=15, swing2LimitDegrees=35, twistLimitDegrees=15}
 {bone1Name=dyn_jacketFront_03_r, bone2Name=dyn_jacketFront_02_r, swing1LimitDegrees=15, swing2LimitDegrees=35, twistLimitDegrees=15}
 {
  bone1Name=dyn_jacketBack_02_r
  bone2Name=dyn_jacketBack_02_l
  swing1LimitDegrees=45
  swing1Motion=Free
  swing2LimitDegrees=45
  swing2Motion=Free
  twistLimitDegrees=45
  twistMotion=Free
 }
 {
  bone1Name=thigh_twist_01_r
  bone2Name=thigh_r
  swing1LimitDegrees=45
  swing1Motion=Free
  swing2LimitDegrees=45
  swing2Motion=Free
  twistLimitDegrees=45
  twistMotion=Free
 }
 {
  bone1Name=dyn_jacketBack_02_r
  bone2Name=dyn_jacketBack_02_l
  swing1LimitDegrees=45
  swing1Motion=Free
  swing2LimitDegrees=45
  swing2Motion=Free
  twistLimitDegrees=45
  twistMotion=Free
 }
 {
  bone1Name=dyn_jacketBack_02_r
  bone2Name=dyn_jacketSide_02_r
  swing1LimitDegrees=45
  swing1Motion=Free
  swing2LimitDegrees=45
  swing2Motion=Free
  twistLimitDegrees=45
  twistMotion=Free
 }
 {
  bone1Name=dyn_jacketBack_02_l
  bone2Name=dyn_jacketSide_02_l
  swing1LimitDegrees=45
  swing1Motion=Free
  swing2LimitDegrees=45
  swing2Motion=Free
  twistLimitDegrees=45
  twistMotion=Free
 }
]
