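# Animation Blueprint for the F_MED_Galactic female character body, adding secondary motion to its coat and jacket bones.
# Thigh constraints and alpha-driven bone offsets adjust the jacket, pouch and holster.
# With rigid body enabled, the cached 'prerigid' pose feeds RigidBody simulations with separate physics assets for skydiving, DBNO and default states.
# Otherwise, a procedural fallback blends crouch-dependent Look At and bone-offset chains into movement-driven Trail controllers.
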AnimBlueprint /BRCosmetics/Characters/Player/Female/Medium/Bodies/F_MED_Galactic/Meshes/F_MED_Galactic_New_AnimBP
target_skeleton=&/F_MED_Galactic_Skeleton
graphs [
 AnimGraph {nodes [
   Blend Poses by bool(bActiveValue=Get bIsCrouching()) {blendType=Linear}
   Blend Poses by bool(BlendPose_0=Use cached pose 'constraints'(), bActiveValue=Get bIsCrouching()) {blendType=Linear}
   n1 = RigidBody!(ComponentPose, Alpha, OverrideWorldGravity:Vector=Get GravityOverride(), ExternalForce) -> (Pose) {
    bClampLinearTranslationLimitToRefPose=true
    bOverrideWorldGravity=true
    overridePhysicsAsset=&/F_MED_Galactic_coat_Physics
    simulationSpace=BaseBoneSpace
    alphaBoolBlend {blendOption=Linear}
    baseBoneRef {boneName=root}
    componentAppliedLinearAccClamp=<3000 3000 3000>
    componentLinearAccScale=<0.25 0.25 0.4>
    componentLinearVelScale=<0.2 0.2 0.2>
   }
   Component To Local(ComponentPose=@n1)
   Component To Local(ComponentPose=Local To Component())
   n2 = NOT Boolean(A=Get bIsCrouching())
   n3 = "Constraint - dyn_jacketFront_inner_mstr_r (R)"!(ComponentPose, bAlphaBoolEnabled:bool=@n2, ConstraintWeights_0=1) -> (Pose) {
    alphaInputType=Bool
    alphaBoolBlend {blendInTime=0.1, blendOption=Linear, blendOutTime=0.1}
    boneToModify {boneName=dyn_jacketFront_inner_mstr_r}
    constraintSetup [{transformType=Rotation, targetBone {boneName=thigh_r}}]
    constraintWeights [1]
   }
   "Constraint - dyn_jacketFront_mstr_r (R)"!(ComponentPose=@n3, bAlphaBoolEnabled:bool=@n2, ConstraintWeights_0=1) -> (Pose) {
    alphaInputType=Bool
    alphaBoolBlend {blendInTime=0.1, blendOption=Linear, blendOutTime=0.1}
    boneToModify {boneName=dyn_jacketFront_mstr_r}
    constraintSetup [{transformType=Rotation, targetBone {boneName=thigh_r}}]
    constraintWeights [1]
   }
   "Transform (Modify) Bone - Bone: dyn_jacketSide_01_l"(Alpha=0, Translation="0,0,0", Rotation="0.000,10,0.000", Scale="1,1,1") {
    rotationMode=BMM_Additive
    rotationSpace=BCS_BoneSpace
    alphaBoolBlend {blendOption=Linear}
    alphaScaleBiasClamp {bInterpResult=true, interpSpeedDecreasing=5}
    boneToModify {boneName=dyn_jacketSide_01_l}
    rotation {yaw=10}
   }
   n4 = Get IsNotCrouchRunning() -> (IsNotCrouchRunning:float)
   "Transform (Modify) Bone - Bone: dyn_jacketBack_03_l"(Alpha=@n4, Translation="0,0,0", Rotation="25,0,0.000", Scale="1,1,1") {
    rotationMode=BMM_Additive
    rotationSpace=BCS_BoneSpace
    alphaBoolBlend {blendOption=Linear}
    alphaScaleBiasClamp {bInterpResult=true, interpSpeedDecreasing=2, interpSpeedIncreasing=5}
    boneToModify {boneName=dyn_jacketBack_03_l}
    rotation {pitch=25}
   }
   n5 = "Transform (Modify) Bone - Bone: dyn_jacketFront_01_r"(Alpha=1, Translation="0,5,5", Rotation="0.000,-25,50", Scale="1,1,1") {
    rotationMode=BMM_Additive
    rotationSpace=BCS_BoneSpace
    translationMode=BMM_Additive
    translationSpace=BCS_BoneSpace
    alphaBoolBlend {blendOption=Linear}
    boneToModify {boneName=dyn_jacketFront_01_r}
    rotation {roll=50, yaw=-25}
    translation {y=5, z=5}
   }
   "Transform (Modify) Bone - Bone: dyn_jacketSide_mstr_r"(ComponentPose=@n5, Alpha=1, Translation="0,5,5", Rotation="0.000,-25,50", Scale="1,1,1") {
    rotationMode=BMM_Additive
    rotationSpace=BCS_BoneSpace
    translationMode=BMM_Additive
    translationSpace=BCS_BoneSpace
    alphaBoolBlend {blendOption=Linear}
    boneToModify {boneName=dyn_jacketSide_mstr_r}
    rotation {roll=50, yaw=-25}
    translation {y=5, z=5}
   }
   n6 = NOT Boolean(A=Get bIsDBNOCarried())
   n7 = "Transform (Modify) Bone - Bone: dyn_jacketFront_inner_mstr_r"!(
    ComponentPose=Local To Component(LocalPose=Use cached pose 'prerigid'())
    bAlphaBoolEnabled:bool=@n6
    Translation="0,-2,-5"
    Rotation="0.000,-35,0.000"
    Scale="1,1,1"
   ) -> (Pose) {
    alphaInputType=Bool
    rotationMode=BMM_Additive
    rotationSpace=BCS_BoneSpace
    translationMode=BMM_Additive
    translationSpace=BCS_BoneSpace
    alphaBoolBlend {blendInTime=0.1, blendOption=Linear, blendOutTime=0.1}
    boneToModify {boneName=dyn_jacketFront_inner_mstr_r}
    rotation {yaw=-35}
    translation {y=-2, z=-5}
   }
   n8 = "Transform (Modify) Bone - Bone: dyn_jacketFront_mstr_r"!(ComponentPose=@n7, bAlphaBoolEnabled:bool=@n6, Translation="0,0,5", Rotation="-35,-15,0.000", Scale="1,1,1") -> (Pose) {
    alphaInputType=Bool
    rotationMode=BMM_Additive
    rotationSpace=BCS_BoneSpace
    translationMode=BMM_Additive
    translationSpace=BCS_BoneSpace
    alphaBoolBlend {blendInTime=0.1, blendOption=Linear, blendOutTime=0.1}
    boneToModify {boneName=dyn_jacketFront_mstr_r}
    rotation {pitch=-35, yaw=-15}
    translation {z=5}
   }
   n9 = "Transform (Modify) Bone - Bone: dyn_jacketFront_inner_mstr_l"!(ComponentPose=@n8, bAlphaBoolEnabled:bool=@n6, Translation="5,5,0", Rotation="0,0,0", Scale="1,1,1") -> (Pose) {
    alphaInputType=Bool
    rotationMode=BMM_Additive
    rotationSpace=BCS_BoneSpace
    translationMode=BMM_Additive
    translationSpace=BCS_BoneSpace
    alphaBoolBlend {blendInTime=0.1, blendOption=Linear, blendOutTime=0.1}
    boneToModify {boneName=dyn_jacketFront_inner_mstr_l}
    translation=<5 5>
   }
   n10 = "Transform (Modify) Bone - Bone: dyn_jacketFront_inner_mstr_r"(ComponentPose=@n9, bAlphaBoolEnabled=Get bIsDBNOCarried(), Translation="0,0,0", Rotation="0,-25,0", Scale="1,1,1") {
    alphaInputType=Bool
    rotationMode=BMM_Additive
    rotationSpace=BCS_BoneSpace
    translationMode=BMM_Additive
    translationSpace=BCS_BoneSpace
    alphaBoolBlend {blendInTime=0.1, blendOption=Linear, blendOutTime=0.1}
    boneToModify {boneName=dyn_jacketFront_inner_mstr_r}
    rotation {yaw=-25}
   }
   n11 = RigidBody(ComponentPose=@n10, OverrideWorldGravity="0,0,-980") {
    bClampLinearTranslationLimitToRefPose=true
    bOverrideWorldGravity=true
    overridePhysicsAsset=&/F_MED_Galactic_coat_DBNOcarry_Physics
    simulationSpace=BaseBoneSpace
    alphaBoolBlend {blendOption=Linear}
    baseBoneRef {boneName=dyn_root_simspace}
    componentAppliedLinearAccClamp=<3000 3000 3000>
    componentLinearAccScale=<0.25 0.25 0.25>
    componentLinearVelScale=<0.2 0.2 0.2>
   }
   n12 = RigidBody(ComponentPose=Local To Component(LocalPose=Use cached pose 'prerigid'()), OverrideWorldGravity=Get GravityOverride()) {
    bClampLinearTranslationLimitToRefPose=true
    bOverrideWorldGravity=true
    overridePhysicsAsset=&F_MED_Galactic_coat_Physics
    simulationSpace=BaseBoneSpace
    alphaBoolBlend {blendOption=Linear}
    baseBoneRef {boneName=dyn_root_simspace}
    componentAppliedLinearAccClamp=<3000 3000 3000>
    componentLinearAccScale=<0.25 0.25 0.25>
    componentLinearVelScale=<0.2 0.2 0.2>
   }
   n13 = Blend Poses by bool(BlendPose_0=Component To Local(ComponentPose=@n11), BlendPose_1=Component To Local(ComponentPose=@n12), bActiveValue=Get bIsDBNO()) {blendType=Linear}
   n14 = "Transform (Modify) Bone - Bone: dyn_jacketBack_mstr_l"(
    ComponentPose=Local To Component(LocalPose=Use cached pose 'prerigid'())
    Alpha=1
    Translation="0,0,2"
    Rotation="15,0.000,0.000"
    Scale="1,1,1"
   ) {
    rotationMode=BMM_Additive
    rotationSpace=BCS_BoneSpace
    translationMode=BMM_Additive
    translationSpace=BCS_BoneSpace
    alphaBoolBlend {blendOption=Linear}
    boneToModify {boneName=dyn_jacketBack_mstr_l}
    rotation {pitch=15}
    translation {z=2}
   }
   n15 = "Transform (Modify) Bone - Bone: dyn_jacketBack_mstr_r"(ComponentPose=@n14, Alpha=1, Translation="0,0,2", Rotation="15,0.000,0.000", Scale="1,1,1") {
    rotationMode=BMM_Additive
    rotationSpace=BCS_BoneSpace
    translationMode=BMM_Additive
    translationSpace=BCS_BoneSpace
    alphaBoolBlend {blendOption=Linear}
    boneToModify {boneName=dyn_jacketBack_mstr_r}
    rotation {pitch=15}
    translation {z=2}
   }
   n16 = "Transform (Modify) Bone - Bone: dyn_jacketSide_mstr_l"(ComponentPose=@n15, Alpha=1, Translation="0,0,2", Rotation="15,0.000,0.000", Scale="1,1,1") {
    rotationMode=BMM_Additive
    rotationSpace=BCS_BoneSpace
    translationMode=BMM_Additive
    translationSpace=BCS_BoneSpace
    alphaBoolBlend {blendOption=Linear}
    boneToModify {boneName=dyn_jacketSide_mstr_l}
    rotation {pitch=15}
    translation {z=2}
   }
   n17 = "Transform (Modify) Bone - Bone: dyn_jacketSide_mstr_r"(ComponentPose=@n16, Alpha=1, Translation="0,0,2", Rotation="15,0.000,0.000", Scale="1,1,1") {
    rotationMode=BMM_Additive
    rotationSpace=BCS_BoneSpace
    translationMode=BMM_Additive
    translationSpace=BCS_BoneSpace
    alphaBoolBlend {blendOption=Linear}
    boneToModify {boneName=dyn_jacketSide_mstr_r}
    rotation {pitch=15}
    translation {z=2}
   }
   n18 = "Transform (Modify) Bone - Bone: dyn_jacketFront_mstr_l"(ComponentPose=@n17, Alpha=1, Translation="0,0,0", Rotation="15,0.000,0.000", Scale="1,1,1") {
    rotationMode=BMM_Additive
    rotationSpace=BCS_BoneSpace
    alphaBoolBlend {blendOption=Linear}
    boneToModify {boneName=dyn_jacketFront_mstr_l}
    rotation {pitch=15}
   }
   n19 = "Transform (Modify) Bone - Bone: dyn_jacketFront_mstr_r"!(ComponentPose=@n18, Alpha:float=1, Translation="0,0,0", Rotation="15,0.000,0.000", Scale="1,1,1") -> (Pose) {
    rotationMode=BMM_Additive
    rotationSpace=BCS_BoneSpace
    alphaBoolBlend {blendOption=Linear}
    boneToModify {boneName=dyn_jacketFront_mstr_r}
    rotation {pitch=15}
   }
   n20 = "Transform (Modify) Bone - Bone: dyn_jacketFront_inner_mstr_r"!(ComponentPose=@n19, Alpha:float=1, Translation="0,0,0", Rotation="15,-10,0.000", Scale="1,1,1") -> (Pose) {
    rotationMode=BMM_Additive
    rotationSpace=BCS_BoneSpace
    alphaBoolBlend {blendOption=Linear}
    boneToModify {boneName=dyn_jacketFront_inner_mstr_r}
    rotation {pitch=15, yaw=-10}
   }
   n21 = "Transform (Modify) Bone - Bone: dyn_jacketFront_inner_mstr_l"!(ComponentPose=@n20, Alpha:float=1, Translation="0,0,0", Rotation="15,10,0.000", Scale="1,1,1") -> (Pose) {
    rotationMode=BMM_Additive
    rotationSpace=BCS_BoneSpace
    alphaBoolBlend {blendOption=Linear}
    boneToModify {boneName=dyn_jacketFront_inner_mstr_l}
    rotation {pitch=15, yaw=10}
   }
   n22 = RigidBody(ComponentPose=@n21, OverrideWorldGravity=Get GravityOverride()) {
    bClampLinearTranslationLimitToRefPose=true
    bOverrideWorldGravity=true
    overridePhysicsAsset=&/F_MED_Galactic_coat_skydive_Physics
    simulationSpace=BaseBoneSpace
    alphaBoolBlend {blendOption=Linear}
    baseBoneRef {boneName=dyn_root_simspace}
    componentAppliedLinearAccClamp=<3000 3000 3000>
    componentLinearAccScale=<0.25 0.25 0.4>
    componentLinearVelScale=<0.2 0.2 0.2>
   }
   n23 = Component To Local(ComponentPose=@n22)
   n24 = Get bIsSkyDiving()
   n25 = Get ThighRotInrR() -> (ThighRotInrR:float)
   n26 = Look At - Bone: dyn_jacketFront_inner_mstr_r(ComponentPose=Local To Component(LocalPose=Use cached pose 'prerigid'()), Alpha=@n25) {
    alphaBoolBlend {blendOption=Linear}
    alphaScaleBiasClamp {bInterpResult=true}
    boneToModify {boneName=dyn_jacketFront_inner_mstr_r}
    lookAtLocation=<0 -12 5>
    lookAtTarget {boneReference {boneName=calf_r}}
    lookAt_Axis {axis=<1 0>}
    lookUp_Axis {axis {x=0}}
   }
   n27 = Look At - Bone: dyn_jacketFront_mstr_r(ComponentPose=@n26, Alpha=@n25) {
    alphaBoolBlend {blendOption=Linear}
    alphaScaleBiasClamp {bInterpResult=true}
    boneToModify {boneName=dyn_jacketFront_mstr_r}
    lookAtLocation=<0 -12 5>
    lookAtTarget {boneReference {boneName=calf_r}}
    lookAt_Axis {axis=<1 0>}
    lookUp_Axis {axis {x=0}}
   }
   n28 = Look At - Bone: dyn_jacketSide_mstr_r(ComponentPose=@n27, Alpha=@n25) {
    alphaBoolBlend {blendOption=Linear}
    alphaScaleBiasClamp {bInterpResult=true}
    boneToModify {boneName=dyn_jacketSide_mstr_r}
    lookAtLocation {x=0, z=11}
    lookAtTarget {boneReference {boneName=calf_r}}
    lookAt_Axis {axis=<1 0>}
    lookUp_Axis {axis {x=0}}
   }
   n29 = "Transform (Modify) Bone - Bone: dyn_jacketFront_01_r"(ComponentPose=@n28, Alpha=1, Translation="0,0,0", Rotation="5,0.000,0.000", Scale="1,1,1") {
    rotationMode=BMM_Additive
    rotationSpace=BCS_BoneSpace
    translationMode=BMM_Additive
    translationSpace=BCS_BoneSpace
    alphaBoolBlend {blendOption=Linear}
    boneToModify {boneName=dyn_jacketFront_01_r}
    rotation {pitch=5}
   }
   n30 = "Transform (Modify) Bone - Bone: dyn_jacketSide_01_r"(ComponentPose=@n29, Alpha=1, Translation="0,0,0", Rotation="10,0.000,0.000", Scale="1,1,1") {
    rotationMode=BMM_Additive
    rotationSpace=BCS_BoneSpace
    translationMode=BMM_Additive
    translationSpace=BCS_BoneSpace
    alphaBoolBlend {blendOption=Linear}
    boneToModify {boneName=dyn_jacketSide_01_r}
    rotation {pitch=10}
   }
   n31 = "Transform (Modify) Bone - Bone: dyn_jacketBack_mstr_r"(ComponentPose=@n30, Alpha=1, Translation="0,0,0", Rotation="10,0.000,0.000", Scale="1,1,1") {
    rotationMode=BMM_Additive
    rotationSpace=BCS_BoneSpace
    translationMode=BMM_Additive
    translationSpace=BCS_BoneSpace
    alphaBoolBlend {blendOption=Linear}
    boneToModify {boneName=dyn_jacketBack_mstr_r}
    rotation {pitch=10}
   }
   n32 = Get ThighRotUpR() -> (ThighRotUpR:float)
   n33 = "Transform (Modify) Bone - Bone: dyn_jacketFront_01_r"(ComponentPose=@n31, Alpha=@n32, Translation="0,0,0", Rotation="9.096,4.987,10.829", Scale="1,1,1") {
    rotationMode=BMM_Additive
    rotationSpace=BCS_BoneSpace
    translationMode=BMM_Additive
    translationSpace=BCS_BoneSpace
    alphaBoolBlend {blendOption=Linear}
    boneToModify {boneName=dyn_jacketFront_01_r}
    rotation=<rot 9.096 4.987 10.829>
   }
   n34 = "Transform (Modify) Bone - Bone: dyn_jacketSide_01_r"(ComponentPose=@n33, Alpha=@n32, Translation="0,0,0", Rotation="5.000,0.000,0.000", Scale="1,1,1") {
    rotationMode=BMM_Additive
    rotationSpace=BCS_BoneSpace
    translationMode=BMM_Additive
    translationSpace=BCS_BoneSpace
    alphaBoolBlend {blendOption=Linear}
    boneToModify {boneName=dyn_jacketSide_01_r}
    rotation {pitch=5}
   }
   n35 = "Transform (Modify) Bone - Bone: dyn_jacketFront_inner_01_r"(ComponentPose=@n34, Alpha=@n32, Translation="0,0,0", Rotation="0,-15,0", Scale="1,1,1") {
    rotationMode=BMM_Additive
    rotationSpace=BCS_BoneSpace
    translationMode=BMM_Additive
    translationSpace=BCS_BoneSpace
    alphaBoolBlend {blendOption=Linear}
    boneToModify {boneName=dyn_jacketFront_inner_01_r}
    rotation {yaw=-15}
   }
   n36 = Get ThighRotInrL() -> (ThighRotInrL:float)
   n37 = Look At - Bone: dyn_jacketFront_inner_mstr_l(ComponentPose=@n35, Alpha=@n36) {
    alphaBoolBlend {blendOption=Linear}
    alphaScaleBiasClamp {bInterpResult=true}
    boneToModify {boneName=dyn_jacketFront_inner_mstr_l}
    lookAtLocation=<0 12 -5>
    lookAtTarget {boneReference {boneName=calf_l}}
    lookAt_Axis {axis=<1 0>}
    lookUp_Axis {axis {x=0}}
   }
   n38 = Look At - Bone: dyn_jacketFront_mstr_l(ComponentPose=@n37, Alpha=@n36) {
    alphaBoolBlend {blendOption=Linear}
    alphaScaleBiasClamp {bInterpResult=true}
    boneToModify {boneName=dyn_jacketFront_mstr_l}
    lookAtLocation=<0 12 -5>
    lookAtTarget {boneReference {boneName=calf_l}}
    lookAt_Axis {axis=<1 0>}
    lookUp_Axis {axis {x=0}}
   }
   n39 = Look At - Bone: dyn_jacketSide_mstr_l(ComponentPose=@n38, Alpha=@n36) {
    alphaBoolBlend {blendOption=Linear}
    alphaScaleBiasClamp {bInterpResult=true}
    boneToModify {boneName=dyn_jacketSide_mstr_l}
    lookAtLocation {x=0, z=-11}
    lookAtTarget {boneReference {boneName=calf_l}}
    lookAt_Axis {axis=<1 0>}
    lookUp_Axis {axis {x=0}}
   }
   n40 = Get ThighRotUpL() -> (ThighRotUpL:float)
   n41 = "Transform (Modify) Bone - Bone: dyn_jacketFront_01_l"(ComponentPose=@n39, Alpha=@n40, Translation="0,0,0", Rotation="9.096,-4.987,-10.828", Scale="1,1,1") {
    rotationMode=BMM_Additive
    rotationSpace=BCS_BoneSpace
    translationMode=BMM_Additive
    translationSpace=BCS_BoneSpace
    alphaBoolBlend {blendOption=Linear}
    boneToModify {boneName=dyn_jacketFront_01_l}
    rotation=<rot 9.096 -4.987 -10.8284>
   }
   n42 = "Transform (Modify) Bone - Bone: dyn_jacketSide_01_l"(ComponentPose=@n41, Alpha=@n40, Translation="0,0,0", Rotation="5,0,0", Scale="1,1,1") {
    rotationMode=BMM_Additive
    rotationSpace=BCS_BoneSpace
    translationMode=BMM_Additive
    translationSpace=BCS_BoneSpace
    alphaBoolBlend {blendOption=Linear}
    boneToModify {boneName=dyn_jacketSide_01_l}
    rotation {pitch=5}
   }
   n43 = "Transform (Modify) Bone - Bone: dyn_jacketFront_inner_01_l"(ComponentPose=@n42, Alpha=@n40, Translation="0,0,0", Rotation="0,15,0", Scale="1,1,1") {
    rotationMode=BMM_Additive
    rotationSpace=BCS_BoneSpace
    translationMode=BMM_Additive
    translationSpace=BCS_BoneSpace
    alphaBoolBlend {blendOption=Linear}
    boneToModify {boneName=dyn_jacketFront_inner_01_l}
    rotation {yaw=15}
   }
   n44 = Get JumpingAlpha() -> (JumpingAlpha:float)
   n45 = "Transform (Modify) Bone - Bone: dyn_jacketBack_01_l"(ComponentPose=@n43, Alpha=@n44, Translation="0,0,0", Rotation="35,10,0.000", Scale="1,1,1") {
    rotationMode=BMM_Additive
    rotationSpace=BCS_BoneSpace
    alphaBoolBlend {blendOption=Linear}
    alphaScaleBiasClamp {bInterpResult=true, interpSpeedDecreasing=2, interpSpeedIncreasing=5}
    boneToModify {boneName=dyn_jacketBack_01_l}
    rotation {pitch=35, yaw=10}
   }
   n46 = "Transform (Modify) Bone - Bone: dyn_jacketBack_01_r"(ComponentPose=@n45, Alpha=@n44, Translation="0,0,0", Rotation="35,-10,0.000", Scale="1,1,1") {
    rotationMode=BMM_Additive
    rotationSpace=BCS_BoneSpace
    alphaBoolBlend {blendOption=Linear}
    alphaScaleBiasClamp {bInterpResult=true, interpSpeedDecreasing=2, interpSpeedIncreasing=5}
    boneToModify {boneName=dyn_jacketBack_01_r}
    rotation {pitch=35, yaw=-10}
   }
   n47 = "Transform (Modify) Bone - Bone: dyn_jacketSide_01_l"(ComponentPose=@n46, Alpha=@n44, Translation="0,0,0", Rotation="15,-15,0.000", Scale="1,1,1") {
    rotationMode=BMM_Additive
    rotationSpace=BCS_BoneSpace
    alphaBoolBlend {blendOption=Linear}
    alphaScaleBiasClamp {bInterpResult=true, interpSpeedDecreasing=2, interpSpeedIncreasing=5}
    boneToModify {boneName=dyn_jacketSide_01_l}
    rotation {pitch=15, yaw=-15}
   }
   n48 = "Transform (Modify) Bone - Bone: dyn_jacketSide_01_l"(ComponentPose=@n47, Alpha=@n44, Translation="0,0,0", Rotation="15,15,0.000", Scale="1,1,1") {
    rotationMode=BMM_Additive
    rotationSpace=BCS_BoneSpace
    alphaBoolBlend {blendOption=Linear}
    alphaScaleBiasClamp {bInterpResult=true, interpSpeedDecreasing=2, interpSpeedIncreasing=5}
    boneToModify {boneName=dyn_jacketSide_01_l}
    rotation {pitch=15, yaw=15}
   }
   n49 = Get SkydiveAlpha() -> (SkydiveAlpha:float)
   n50 = "Transform (Modify) Bone - Bone: dyn_jacketBack_mstr_l"(ComponentPose=@n48, Alpha=@n49, Translation="0,0,0", Rotation="15,0,0.000", Scale="1,1,1") {
    rotationMode=BMM_Additive
    rotationSpace=BCS_BoneSpace
    alphaBoolBlend {blendOption=Linear}
    alphaScaleBiasClamp {bInterpResult=true, interpSpeedDecreasing=2, interpSpeedIncreasing=5}
    boneToModify {boneName=dyn_jacketBack_mstr_l}
    rotation {pitch=15}
   }
   n51 = "Transform (Modify) Bone - Bone: dyn_jacketBack_03_l"(ComponentPose=@n50, Alpha=@n49, Translation="0,0,0", Rotation="20,0,0.000", Scale="1,1,1") {
    rotationMode=BMM_Additive
    rotationSpace=BCS_BoneSpace
    alphaBoolBlend {blendOption=Linear}
    alphaScaleBiasClamp {bInterpResult=true, interpSpeedDecreasing=2, interpSpeedIncreasing=5}
    boneToModify {boneName=dyn_jacketBack_03_l}
    rotation {pitch=20}
   }
   n52 = "Transform (Modify) Bone - Bone: dyn_jacketBack_mstr_r"(ComponentPose=@n51, Alpha=@n49, Translation="0,0,0", Rotation="15,0,0.000", Scale="1,1,1") {
    rotationMode=BMM_Additive
    rotationSpace=BCS_BoneSpace
    alphaBoolBlend {blendOption=Linear}
    alphaScaleBiasClamp {bInterpResult=true, interpSpeedDecreasing=2, interpSpeedIncreasing=5}
    boneToModify {boneName=dyn_jacketBack_mstr_r}
    rotation {pitch=15}
   }
   n53 = Get ParachuteAlpha() -> (ParachuteAlpha:float)
   n54 = "Transform (Modify) Bone - Bone: dyn_jacketBack_01_l"(ComponentPose=@n52, Alpha=@n53, Translation="0,0,0", Rotation="15,10,0.000", Scale="1,1,1") {
    rotationMode=BMM_Additive
    rotationSpace=BCS_BoneSpace
    alphaBoolBlend {blendOption=Linear}
    alphaScaleBiasClamp {bInterpResult=true, interpSpeedDecreasing=2, interpSpeedIncreasing=5}
    boneToModify {boneName=dyn_jacketBack_01_l}
    rotation {pitch=15, yaw=10}
   }
   n55 = "Transform (Modify) Bone - Bone: dyn_jacketBack_01_r"(ComponentPose=@n54, Alpha=@n53, Translation="0,0,0", Rotation="15,-10,0.000", Scale="1,1,1") {
    rotationMode=BMM_Additive
    rotationSpace=BCS_BoneSpace
    alphaBoolBlend {blendOption=Linear}
    alphaScaleBiasClamp {bInterpResult=true, interpSpeedDecreasing=2, interpSpeedIncreasing=5}
    boneToModify {boneName=dyn_jacketBack_01_r}
    rotation {pitch=15, yaw=-10}
   }
   n56 = "Transform (Modify) Bone - Bone: dyn_jacketSide_01_l"(ComponentPose=@n55, Alpha=@n53, Translation="0,0,0", Rotation="15,-15,0.000", Scale="1,1,1") {
    rotationMode=BMM_Additive
    rotationSpace=BCS_BoneSpace
    alphaBoolBlend {blendOption=Linear}
    alphaScaleBiasClamp {bInterpResult=true, interpSpeedDecreasing=2, interpSpeedIncreasing=5}
    boneToModify {boneName=dyn_jacketSide_01_l}
    rotation {pitch=15, yaw=-15}
   }
   n57 = "Transform (Modify) Bone - Bone: dyn_jacketSide_01_l"(ComponentPose=@n56, Alpha=@n53, Translation="0,0,0", Rotation="15,15,0.000", Scale="1,1,1") {
    rotationMode=BMM_Additive
    rotationSpace=BCS_BoneSpace
    alphaBoolBlend {blendOption=Linear}
    alphaScaleBiasClamp {bInterpResult=true, interpSpeedDecreasing=2, interpSpeedIncreasing=5}
    boneToModify {boneName=dyn_jacketSide_01_l}
    rotation {pitch=15, yaw=15}
   }
   n58 = Component To Local(ComponentPose=@n57)
   n59 = "Transform (Modify) Bone - Bone: dyn_jacketFront_01_r"(ComponentPose=Local To Component(LocalPose=Use cached pose 'prerigid'()), Alpha=1, Translation="0,0,0", Rotation="15,0,0", Scale="1,1,1") {
    rotationMode=BMM_Additive
    rotationSpace=BCS_BoneSpace
    translationMode=BMM_Additive
    translationSpace=BCS_BoneSpace
    alphaBoolBlend {blendOption=Linear}
    boneToModify {boneName=dyn_jacketFront_01_r}
    rotation {pitch=15}
   }
   n60 = "Transform (Modify) Bone - Bone: dyn_jacketSide_01_r"(ComponentPose=@n59, Alpha=1, Translation="0,0,0", Rotation="20,0,0", Scale="1,1,1") {
    rotationMode=BMM_Additive
    rotationSpace=BCS_BoneSpace
    translationMode=BMM_Additive
    translationSpace=BCS_BoneSpace
    alphaBoolBlend {blendOption=Linear}
    boneToModify {boneName=dyn_jacketSide_01_r}
    rotation {pitch=20}
   }
   n61 = "Transform (Modify) Bone - Bone: dyn_jacketBack_01_l"(ComponentPose=@n60, Alpha=@n4, Translation="0,0,0", Rotation="15,0,0.000", Scale="1,1,1") {
    rotationMode=BMM_Additive
    rotationSpace=BCS_BoneSpace
    alphaBoolBlend {blendOption=Linear}
    alphaScaleBiasClamp {bInterpResult=true, interpSpeedDecreasing=5}
    boneToModify {boneName=dyn_jacketBack_01_l}
    rotation {pitch=15}
   }
   n62 = "Transform (Modify) Bone - Bone: dyn_jacketBack_01_r"(ComponentPose=@n61, Alpha=@n4, Translation="0,0,0", Rotation="10,-10,0.000", Scale="1,1,1") {
    rotationMode=BMM_Additive
    rotationSpace=BCS_BoneSpace
    alphaBoolBlend {blendOption=Linear}
    alphaScaleBiasClamp {bInterpResult=true, interpSpeedDecreasing=5}
    boneToModify {boneName=dyn_jacketBack_01_r}
    rotation {pitch=10, yaw=-10}
   }
   n63 = "Transform (Modify) Bone - Bone: dyn_jacketBack_02_r"(ComponentPose=@n62, Alpha=@n4, Translation="0,0,0", Rotation="30,0,0.000", Scale="1,1,1") {
    rotationMode=BMM_Additive
    rotationSpace=BCS_BoneSpace
    alphaBoolBlend {blendOption=Linear}
    alphaScaleBiasClamp {bInterpResult=true, interpSpeedDecreasing=5}
    boneToModify {boneName=dyn_jacketBack_02_r}
    rotation {pitch=30}
   }
   n64 = Get ThighRotUpR()
   n65 = "Transform (Modify) Bone - Bone: dyn_jacketFront_inner_mstr_r"(ComponentPose=@n63, Alpha=@n64, Translation="0,0,3", Rotation="15,-30.000,0.000", Scale="1,1,1") {
    rotationMode=BMM_Additive
    rotationSpace=BCS_BoneSpace
    translationMode=BMM_Additive
    translationSpace=BCS_BoneSpace
    alphaBoolBlend {blendOption=Linear}
    alphaScaleBiasClamp {bMapRange=true, outRange {max=0, min=1}}
    boneToModify {boneName=dyn_jacketFront_inner_mstr_r}
    rotation {pitch=15, yaw=-30}
    translation {z=3}
   }
   n66 = "Transform (Modify) Bone - Bone: dyn_jacketBack_01_r"(ComponentPose=@n65, Alpha=@n64, Translation="0,0,0", Rotation="15,0,0", Scale="1,1,1") {
    rotationMode=BMM_Additive
    rotationSpace=BCS_BoneSpace
    alphaBoolBlend {blendOption=Linear}
    alphaScaleBiasClamp {bInterpResult=true, interpSpeedDecreasing=5}
    boneToModify {boneName=dyn_jacketBack_01_r}
    rotation {pitch=15}
   }
   n67 = Get ThighRotOtrR() -> (ThighRotOtrR:float)
   n68 = "Transform (Modify) Bone - Bone: dyn_jacketFront_01_r"(ComponentPose=@n66, Alpha=@n67, Translation="0,0,0", Rotation="30,15.000,-0.000", Scale="1,1,1") {
    rotationMode=BMM_Additive
    rotationSpace=BCS_BoneSpace
    translationMode=BMM_Additive
    translationSpace=BCS_BoneSpace
    alphaBoolBlend {blendOption=Linear}
    boneToModify {boneName=dyn_jacketFront_01_r}
    rotation {pitch=30, yaw=15}
   }
   n69 = "Transform (Modify) Bone - Bone: dyn_jacketFront_02_r"(ComponentPose=@n68, Alpha=@n67, Translation="0,0,0", Rotation="-30,0,0", Scale="1,1,1") {
    rotationMode=BMM_Additive
    rotationSpace=BCS_BoneSpace
    translationMode=BMM_Additive
    translationSpace=BCS_BoneSpace
    alphaBoolBlend {blendOption=Linear}
    boneToModify {boneName=dyn_jacketFront_02_r}
    rotation {pitch=-30}
   }
   n70 = "Transform (Modify) Bone - Bone: dyn_jacketSide_02_r"(ComponentPose=@n69, Alpha=@n67, Translation="0,0,0", Rotation="45,0,0", Scale="1,1,1") {
    rotationMode=BMM_Additive
    rotationSpace=BCS_BoneSpace
    translationMode=BMM_Additive
    translationSpace=BCS_BoneSpace
    alphaBoolBlend {blendOption=Linear}
    boneToModify {boneName=dyn_jacketSide_02_r}
    rotation {pitch=45}
   }
   n71 = Get ThighRotUpL()
   n72 = "Transform (Modify) Bone - Bone: dyn_jacketFront_01_l"(ComponentPose=@n70, Alpha=@n71, Translation="0,0,0", Rotation="-30,-15.000,0", Scale="1,1,1") {
    rotationMode=BMM_Additive
    rotationSpace=BCS_BoneSpace
    translationMode=BMM_Additive
    translationSpace=BCS_BoneSpace
    alphaBoolBlend {blendOption=Linear}
    boneToModify {boneName=dyn_jacketFront_01_l}
    rotation {pitch=-30, yaw=-15}
   }
   n73 = "Transform (Modify) Bone - Bone: dyn_jacketFront_02_l"(ComponentPose=@n72, Alpha=@n71, Translation="0,0,0", Rotation="-30,0,0", Scale="1,1,1") {
    rotationMode=BMM_Additive
    rotationSpace=BCS_BoneSpace
    translationMode=BMM_Additive
    translationSpace=BCS_BoneSpace
    alphaBoolBlend {blendOption=Linear}
    boneToModify {boneName=dyn_jacketFront_02_l}
    rotation {pitch=-30}
   }
   n74 = "Transform (Modify) Bone - Bone: dyn_jacketSide_01_l"(ComponentPose=@n73, Alpha=@n71, Translation="0,-2,3", Rotation="0,0,0", Scale="1,1,1") {
    rotationMode=BMM_Additive
    rotationSpace=BCS_BoneSpace
    translationMode=BMM_Additive
    translationSpace=BCS_BoneSpace
    alphaBoolBlend {blendOption=Linear}
    boneToModify {boneName=dyn_jacketSide_01_l}
    translation {y=-2, z=3}
   }
   n75 = "Transform (Modify) Bone - Bone: dyn_jacketSide_02_l"(ComponentPose=@n74, Alpha=@n71, Translation="0,0,0", Rotation="-30,0,0", Scale="1,1,1") {
    rotationMode=BMM_Additive
    rotationSpace=BCS_BoneSpace
    translationMode=BMM_Additive
    translationSpace=BCS_BoneSpace
    alphaBoolBlend {blendOption=Linear}
    boneToModify {boneName=dyn_jacketSide_02_l}
    rotation {pitch=-30}
   }
   n76 = Get ThighRotOtrL() -> (ThighRotOtrL:float)
   n77 = "Transform (Modify) Bone - Bone: dyn_jacketFront_01_l"(ComponentPose=@n75, Alpha=@n76, Translation="0,0,0", Rotation="45,0,0", Scale="1,1,1") {
    rotationMode=BMM_Additive
    rotationSpace=BCS_BoneSpace
    translationMode=BMM_Additive
    translationSpace=BCS_BoneSpace
    alphaBoolBlend {blendOption=Linear}
    boneToModify {boneName=dyn_jacketFront_01_l}
    rotation {pitch=45}
   }
   n78 = "Transform (Modify) Bone - Bone: dyn_jacketFront_02_l"(ComponentPose=@n77, Alpha=@n76, Translation="0,0,2", Rotation="30,0,0", Scale="1,1,1") {
    rotationMode=BMM_Additive
    rotationSpace=BCS_BoneSpace
    translationMode=BMM_Additive
    translationSpace=BCS_BoneSpace
    alphaBoolBlend {blendOption=Linear}
    boneToModify {boneName=dyn_jacketFront_02_l}
    rotation {pitch=30}
    translation {z=2}
   }
   n79 = "Transform (Modify) Bone - Bone: dyn_jacketSide_02_l"(ComponentPose=@n78, Alpha=@n76, Translation="0,0,0", Rotation="45,0,0", Scale="1,1,1") {
    rotationMode=BMM_Additive
    rotationSpace=BCS_BoneSpace
    translationMode=BMM_Additive
    translationSpace=BCS_BoneSpace
    alphaBoolBlend {blendOption=Linear}
    boneToModify {boneName=dyn_jacketSide_02_l}
    rotation {pitch=45}
   }
   n80 = Component To Local(ComponentPose=@n79)
   n81 = Get bIsCrouching()
   n82 = Get BwdMovement() -> (BwdMovement:float)
   n83 = Trail controller - Bone: dyn_jacketBack_03_r!(
    ComponentPose=Local To Component(LocalPose=Blend Poses by bool(BlendPose_0=@n80, BlendPose_1=@n58, bActiveValue=@n81) {blendType=Linear})
    Alpha=@n82
    FakeVelocity:Vector="0,0,0"
   ) -> (Pose) {
    bActorSpaceFakeVel=true
    bShowTrailLocation=true
    debugLifeTime=0.1
    maxDeltaTime=0.0333
    relaxationSpeedScale=3
    alphaScaleBiasClamp {bInterpResult=true}
    baseJoint {boneName=root}
    planarLimits [
     {drivingBone {boneName=root}, planeTransform {location=<0 0 10>, rotation=<rot 0 0 0>, scale=<1 1 1>}}
     {drivingBone {boneName=spine_03}, planeTransform {location=<0 -20 0>, rotation=<rot 0 15 -89.9999>, scale=<1 1 1>}}
    ]
    rotationLimits [{limitMax=<180 180 180>}, {limitMax=<180 180 180>}]
    rotationOffsets [<0 0 0>, <0 0 0>]
    trailBone {boneName=dyn_jacketBack_03_r}
    trailRelaxationSpeed {editorCurveData {keys [{arriveTangent=-72.0739, interpMode=RCIM_Cubic, leaveTangent=-72.0739, tangentMode=RCTM_User, time=0, value=35}, <at 1 5 RCIM_Cubic>]}}
   }
   n84 = Trail controller - Bone: dyn_jacketBack_03_l!(ComponentPose=@n83, Alpha=@n82, FakeVelocity:Vector="0,0,0") -> (Pose) {
    bActorSpaceFakeVel=true
    bShowTrailLocation=true
    debugLifeTime=0.1
    maxDeltaTime=0.0333
    relaxationSpeedScale=3
    alphaScaleBiasClamp {bInterpResult=true}
    baseJoint {boneName=root}
    planarLimits [
     {drivingBone {boneName=root}, planeTransform {location=<0 0 10>, rotation=<rot 0 0 0>, scale=<1 1 1>}}
     {drivingBone {boneName=spine_03}, planeTransform {location=<0 -20 0>, rotation=<rot 0 15 -89.9999>, scale=<1 1 1>}}
    ]
    rotationLimits [{limitMax=<180 180 180>}, {limitMax=<180 180 180>}]
    rotationOffsets [<0 0 0>, <0 0 0>]
    trailBone {boneName=dyn_jacketBack_03_l}
    trailRelaxationSpeed {editorCurveData {keys [{arriveTangent=-72.0739, interpMode=RCIM_Cubic, leaveTangent=-72.0739, tangentMode=RCTM_User, time=0, value=35}, <at 1 5 RCIM_Cubic>]}}
   }
   n85 = Get LeftMovement() -> (LeftMovement:float)
   n86 = Trail controller - Bone: dyn_jacketFront_03_r!(ComponentPose=@n84, Alpha=@n85, FakeVelocity:Vector="0,0,0") -> (Pose) {
    bActorSpaceFakeVel=true
    bShowTrailLocation=true
    debugLifeTime=0.1
    maxDeltaTime=0.0333
    relaxationSpeedScale=3
    alphaScaleBiasClamp {bInterpResult=true}
    baseJoint {boneName=root}
    planarLimits [
     {drivingBone {boneName=root}, planeTransform {location=<0 0 10>, rotation=<rot 0 0 0>, scale=<1 1 1>}}
     {drivingBone {boneName=spine_03}, planeTransform {location=<0 -20 0>, rotation=<rot 0 15 -89.9999>, scale=<1 1 1>}}
    ]
    rotationLimits [{limitMax=<180 180 180>}, {limitMax=<180 180 180>}]
    rotationOffsets [<0 0 0>, <0 0 0>]
    trailBone {boneName=dyn_jacketFront_03_r}
    trailRelaxationSpeed {editorCurveData {keys [{arriveTangent=-72.0739, interpMode=RCIM_Cubic, leaveTangent=-72.0739, tangentMode=RCTM_User, time=0, value=35}, <at 1 5 RCIM_Cubic>]}}
   }
   n87 = Trail controller - Bone: dyn_jacketSide_03_r!(ComponentPose=@n86, Alpha=@n85, FakeVelocity:Vector="0,0,0") -> (Pose) {
    bActorSpaceFakeVel=true
    bShowTrailLocation=true
    debugLifeTime=0.1
    maxDeltaTime=0.0333
    relaxationSpeedScale=3
    alphaScaleBiasClamp {bInterpResult=true}
    baseJoint {boneName=root}
    planarLimits [
     {drivingBone {boneName=root}, planeTransform {location=<0 0 10>, rotation=<rot 0 0 0>, scale=<1 1 1>}}
     {drivingBone {boneName=spine_03}, planeTransform {location=<0 -20 0>, rotation=<rot 0 15 -89.9999>, scale=<1 1 1>}}
    ]
    rotationLimits [{limitMax=<180 180 180>}, {limitMax=<180 180 180>}]
    rotationOffsets [<0 0 0>, <0 0 0>]
    trailBone {boneName=dyn_jacketSide_03_r}
    trailRelaxationSpeed {editorCurveData {keys [{arriveTangent=-72.0739, interpMode=RCIM_Cubic, leaveTangent=-72.0739, tangentMode=RCTM_User, time=0, value=35}, <at 1 5 RCIM_Cubic>]}}
   }
   n88 = Get RightMovement() -> (RightMovement:float)
   n89 = Trail controller - Bone: dyn_jacketFront_03_l!(ComponentPose=@n87, Alpha=@n88, FakeVelocity:Vector="0,0,0") -> (Pose) {
    bActorSpaceFakeVel=true
    bShowTrailLocation=true
    debugLifeTime=0.1
    maxDeltaTime=0.0333
    relaxationSpeedScale=3
    alphaScaleBiasClamp {bInterpResult=true}
    baseJoint {boneName=root}
    planarLimits [
     {drivingBone {boneName=root}, planeTransform {location=<0 0 10>, rotation=<rot 0 0 0>, scale=<1 1 1>}}
     {drivingBone {boneName=spine_03}, planeTransform {location=<0 -20 0>, rotation=<rot 0 15 -89.9999>, scale=<1 1 1>}}
    ]
    rotationLimits [{limitMax=<180 180 180>}, {limitMax=<180 180 180>}]
    rotationOffsets [<0 0 0>, <0 0 0>]
    trailBone {boneName=dyn_jacketFront_03_l}
    trailRelaxationSpeed {editorCurveData {keys [{arriveTangent=-72.0739, interpMode=RCIM_Cubic, leaveTangent=-72.0739, tangentMode=RCTM_User, time=0, value=35}, <at 1 5 RCIM_Cubic>]}}
   }
   n90 = Trail controller - Bone: dyn_jacketSide_03_l!(ComponentPose=@n89, Alpha=@n88, FakeVelocity:Vector="0,0,0") -> (Pose) {
    bActorSpaceFakeVel=true
    bShowTrailLocation=true
    debugLifeTime=0.1
    maxDeltaTime=0.0333
    relaxationSpeedScale=3
    alphaScaleBiasClamp {bInterpResult=true}
    baseJoint {boneName=root}
    planarLimits [
     {drivingBone {boneName=root}, planeTransform {location=<0 0 10>, rotation=<rot 0 0 0>, scale=<1 1 1>}}
     {drivingBone {boneName=spine_03}, planeTransform {location=<0 -20 0>, rotation=<rot 0 15 -89.9999>, scale=<1 1 1>}}
    ]
    rotationLimits [{limitMax=<180 180 180>}, {limitMax=<180 180 180>}]
    rotationOffsets [<0 0 0>, <0 0 0>]
    trailBone {boneName=dyn_jacketSide_03_l}
    trailRelaxationSpeed {editorCurveData {keys [{arriveTangent=-72.0739, interpMode=RCIM_Cubic, leaveTangent=-72.0739, tangentMode=RCTM_User, time=0, value=35}, <at 1 5 RCIM_Cubic>]}}
   }
   Output Pose(Result=Blend Poses by bool(
     BlendPose_0=Blend Poses by bool(BlendPose_0=@n23, BlendPose_1=@n13, bActiveValue=@n24) {blendType=Linear}
     BlendPose_1=Component To Local(ComponentPose=@n90)
     bActiveValue=Get bRigidBodyEnabled()
    ) {blendType=Linear}) {layerGroup=None}
   n91 = "Constraint - dyn_jacketFront_inner_mstr_l (R)"!(ComponentPose=Local To Component(LocalPose=Use cached pose 'constraints'()), bAlphaBoolEnabled:bool=@n2, ConstraintWeights_0=1) -> (Pose) {
    alphaInputType=Bool
    alphaBoolBlend {blendInTime=0.1, blendOption=Linear, blendOutTime=0.1}
    boneToModify {boneName=dyn_jacketFront_inner_mstr_l}
    constraintSetup [{transformType=Rotation, targetBone {boneName=thigh_l}}]
    constraintWeights [1]
   }
   n92 = "Constraint - dyn_jacketFront_mstr_l (R)"!(ComponentPose=@n91, bAlphaBoolEnabled:bool=@n2, ConstraintWeights_0=1) -> (Pose) {
    alphaInputType=Bool
    alphaBoolBlend {blendInTime=0.1, blendOption=Linear, blendOutTime=0.1}
    boneToModify {boneName=dyn_jacketFront_mstr_l}
    constraintSetup [{transformType=Rotation, targetBone {boneName=thigh_l}}]
    constraintWeights [1]
   }
   n93 = Get ThighRotL() -> (ThighRotL:float)
   n94 = "Transform (Modify) Bone - Bone: dyn_jacketFront_mstr_l"(ComponentPose=@n92, Alpha=@n93, Translation="-7,-5,5", Rotation="35,-35,-25", Scale="1,1,1") {
    rotationMode=BMM_Additive
    rotationSpace=BCS_BoneSpace
    scaleMode=BMM_Additive
    scaleSpace=BCS_BoneSpace
    translationMode=BMM_Additive
    translationSpace=BCS_BoneSpace
    alphaBoolBlend {blendOption=Linear}
    boneToModify {boneName=dyn_jacketFront_mstr_l}
    rotation=<rot 35 -35 -25>
    translation=<-7 -5 5>
   }
   n95 = "Transform (Modify) Bone - Bone: dyn_jacketFront_inner_mstr_l"(ComponentPose=@n94, Alpha=@n93, Translation="-15,-2,12", Rotation="15,-15,-45", Scale="1,1,1") {
    rotationMode=BMM_Additive
    rotationSpace=BCS_BoneSpace
    scaleMode=BMM_Additive
    scaleSpace=BCS_BoneSpace
    translationMode=BMM_Additive
    translationSpace=BCS_BoneSpace
    alphaBoolBlend {blendOption=Linear}
    boneToModify {boneName=dyn_jacketFront_inner_mstr_l}
    rotation=<rot 15 -15 -45>
    translation=<-15 -2 12>
   }
   n96 = "Transform (Modify) Bone - Bone: dyn_jacketSide_mstr_l"(ComponentPose=@n95, Alpha=@n93, Translation="-2,-4,0", Rotation="15,-10,0", Scale="1,1,1") {
    rotationMode=BMM_Additive
    rotationSpace=BCS_BoneSpace
    scaleMode=BMM_Additive
    scaleSpace=BCS_BoneSpace
    translationMode=BMM_Additive
    translationSpace=BCS_BoneSpace
    alphaBoolBlend {blendOption=Linear}
    boneToModify {boneName=dyn_jacketSide_mstr_l}
    rotation {pitch=15, yaw=-10}
    translation=<-2 -4>
   }
   n97 = Get ThighRotR() -> (ThighRotR:float)
   n98 = "Transform (Modify) Bone - Bone: dyn_jacketFront_mstr_r"(ComponentPose=@n96, Alpha=@n97, Translation="-7,-0.000,5", Rotation="35,35,25", Scale="1,1,1") {
    rotationMode=BMM_Additive
    rotationSpace=BCS_BoneSpace
    scaleMode=BMM_Additive
    scaleSpace=BCS_BoneSpace
    translationMode=BMM_Additive
    translationSpace=BCS_BoneSpace
    alphaBoolBlend {blendOption=Linear}
    boneToModify {boneName=dyn_jacketFront_mstr_r}
    rotation=<rot 35 35 25>
    translation {x=-7, z=5}
   }
   n99 = "Transform (Modify) Bone - Bone: dyn_jacketFront_inner_mstr_r"(ComponentPose=@n98, Alpha=@n97, Translation="-10,2,15", Rotation="25,15,35", Scale="1,1,1") {
    rotationMode=BMM_Additive
    rotationSpace=BCS_BoneSpace
    scaleMode=BMM_Additive
    scaleSpace=BCS_BoneSpace
    translationMode=BMM_Additive
    translationSpace=BCS_BoneSpace
    alphaBoolBlend {blendOption=Linear}
    boneToModify {boneName=dyn_jacketFront_inner_mstr_r}
    rotation=<rot 25 15 35>
    translation=<-10 2 15>
   }
   n100 = "Transform (Modify) Bone - Bone: dyn_jacketSide_mstr_r"(ComponentPose=@n99, Alpha=@n97, Translation="-2,2,0", Rotation="15,10,0", Scale="1,1,1") {
    rotationMode=BMM_Additive
    rotationSpace=BCS_BoneSpace
    scaleMode=BMM_Additive
    scaleSpace=BCS_BoneSpace
    translationMode=BMM_Additive
    translationSpace=BCS_BoneSpace
    alphaBoolBlend {blendOption=Linear}
    boneToModify {boneName=dyn_jacketSide_mstr_r}
    rotation {pitch=15, yaw=10}
    translation=<-2 2>
   }
   n101 = "Transform (Modify) Bone - Bone: dyn_pouch_01"!(
    ComponentPose=@n100
    bAlphaBoolEnabled:bool=Get bIsCrouching()
    Translation="-2.605,-5.489,-0.699"
    Rotation="9.364,-19.217,-51.763"
    Scale="1,1,1"
   ) -> (Pose) {
    alphaInputType=Bool
    rotationMode=BMM_Additive
    rotationSpace=BCS_BoneSpace
    translationMode=BMM_Additive
    translationSpace=BCS_BoneSpace
    alphaBoolBlend {blendInTime=0.1, blendOption=Linear, blendOutTime=0.1}
    boneToModify {boneName=dyn_pouch_01}
    rotation=<rot 9.364 -19.217 -51.7631>
    translation=<-2.6049 -5.4891 -0.699>
   }
   n102 = "Transform (Modify) Bone - Bone: dyn_holster_MSTR"(ComponentPose=@n101, Alpha=1, Translation="1,0.000,0.000", Rotation="0,0,0", Scale="1,1,1") {translationMode=BMM_Additive, alphaBoolBlend {blendOption=Linear}, boneToModify {boneName=dyn_holster_MSTR}, translation {x=1}}
   n103 = "Transform (Modify) Bone - Bone: dyn_floor_COL"!(ComponentPose=@n102, bAlphaBoolEnabled:bool=Get bIsFalling(), Translation="0,0.000,-300", Rotation="0,0,0", Scale="1,1,1") -> (Pose) {
    alphaInputType=Bool
    translationMode=BMM_Additive
    translationSpace=BCS_BoneSpace
    alphaBoolBlend {blendInTime=0.1, blendOption=Linear, blendOutTime=0.5}
    boneToModify {boneName=dyn_floor_COL}
    translation {z=-300}
   }
   Save cached pose 'prerigid'(Pose=Component To Local(ComponentPose=@n103)) {cacheName=prerigid}
   n104 = "Constraint - dyn_jacketFront_inner_mstr_r (PPPPP)"!(
    ComponentPose=Local To Component(LocalPose=Copy Pose From Mesh() {bUseAttachedParent=true})
    Alpha=1
    ConstraintWeights_0=0.0229
    ConstraintWeights_1:float=0.0134
    ConstraintWeights_2:float=0.9075
    ConstraintWeights_3:float=0.0563
    ConstraintWeights_4:float=0
   ) -> (Pose) {
    alphaBoolBlend {blendOption=Linear}
    boneToModify {boneName=dyn_jacketFront_inner_mstr_r}
    constraintSetup [
     {targetBone {boneName=spine_02}}
     {targetBone {boneName=spine_01}}
     {targetBone {boneName=pelvis}}
     {targetBone {boneName=thigh_twist_01_r}}
     {targetBone {boneName=dyn_jacketFront_inner_01_r}}
    ]
    constraintWeights [1,1,1,1,1]
   }
   n105 = "Constraint - dyn_jacketFront_inner_mstr_l (PPPPPP)"!(
    ComponentPose=@n104
    Alpha=1
    ConstraintWeights_0=0.0226
    ConstraintWeights_1:float=0.0034
    ConstraintWeights_2:float=0.013
    ConstraintWeights_3:float=0.9097
    ConstraintWeights_4:float=0.0514
    ConstraintWeights_5:float=0
   ) -> (Pose) {
    alphaBoolBlend {blendOption=Linear}
    boneToModify {boneName=dyn_jacketFront_inner_mstr_l}
    constraintSetup [
     {targetBone {boneName=spine_02}}
     {targetBone {boneName=spine_04}}
     {targetBone {boneName=spine_01}}
     {targetBone {boneName=pelvis}}
     {targetBone {boneName=thigh_twist_01_l}}
     {targetBone {boneName=dyn_jacketFront_inner_01_l}}
    ]
    constraintWeights [1,1,1,1,1,1]
   }
   n106 = "Constraint - dyn_pouch_MSTR (PPPPPPP)"!(
    ComponentPose=@n105
    Alpha=1
    ConstraintWeights_0=0.987
    ConstraintWeights_1:float=0.0039
    ConstraintWeights_2:float=0.0049
    ConstraintWeights_3:float=0.0029
    ConstraintWeights_4:float=0.0012
    ConstraintWeights_5:float=0
    ConstraintWeights_6:float=0
   ) -> (Pose) {
    alphaBoolBlend {blendOption=Linear}
    boneToModify {boneName=dyn_pouch_MSTR}
    constraintSetup [
     {targetBone {boneName=pelvis}}
     {targetBone {boneName=spine_01}}
     {targetBone {boneName=spine_02}}
     {targetBone {boneName=spine_03}}
     {targetBone {boneName=spine_04}}
     {targetBone {boneName=spine_05}}
     {targetBone {boneName=clavicle_l}}
    ]
    constraintWeights [1,1,1,1,1,1,1]
   }
   Save cached pose 'constraints'(Pose=Component To Local(ComponentPose=@n106)) {cacheName=constraints}
   Get ThighRotR_0() -> (ThighRotR_0:double)
  ]}
 EventGraph {nodes [
   Event BlueprintUpdateAnimation()
   > n1 = Sequence!(execute) -> (then_0, then_1, then_2:Exec, then_3:Exec, then_4:Exec)
   n2 = ComplexSimSpaceFunction DEPRECATED!(
    execute=@n1.then_1
    CharacterPartAnimInstance:Custom Character Part Anim Instance Object Reference=Self-Reference()
    InterpSpeedMax:double=9
    PawnSpeedMax:double=20
    RInterpErrorTolerance:double=1
    DeltaTime:double=0
    DebugDraw:bool=false
    sim_space_rotation_in:Rotator=Get sim_space_rotation() -> (sim_space_rotation:Rotator)
    inComponentSpaceIn:bool=Get InComponentSpace() -> (InComponentSpace:bool)
    inComponentSpaceLastFrameIn:bool=Get InComponentSpaceLastFrame() -> (InComponentSpaceLastFrame:bool)
    lockedOnSimSpaceIn:bool=Get LockedOnSimSpace() -> (LockedOnSimSpace:bool)
    InterpSpeedIn:double=Get InterpSpeed() -> (InterpSpeed:double)
   ) -> (
    then
    sim_space_rotation:Rotator
    inComponentSpaceOut:bool
    inComponentSpaceLastFrameOut:bool
    lockedOnSimSpaceOut:bool
    InterpSpeedOut:double
   )
   > Set sim_space_rotation(sim_space_rotation:Rotator=@n2) -> (Output_Get:Rotator)
   > Set InComponentSpace(InComponentSpace:bool=@n2.inComponentSpaceOut) -> (Output_Get:bool)
   > Set InComponentSpaceLastFrame(InComponentSpaceLastFrame:bool=@n2.inComponentSpaceLastFrameOut) -> (Output_Get:bool)
   > Set LockedOnSimSpace(LockedOnSimSpace:bool=@n2.lockedOnSimSpaceOut) -> (Output_Get:bool)
   > Set InterpSpeed(InterpSpeed:double=@n2.InterpSpeedOut) -> (Output_Get:double)
   > n3 = Gravity Override State Function DEPRECATED(CharacterPartAnimInstance=Self-Reference(), GameTimeInSeconds=GetGameTimeInSeconds(), GravityOverrideParams=Get Gravity Override Params())
   > Set GravityOverride(GravityOverride=@n3)
   n4 = PlayerMovement(execute=@n1.then_3, AnimInstance=Self-Reference(), XDotProductCoefficient=1, YDotProductCoefficient=1, ZDotProductCoefficient=1)
   > UpdatePostPlayerMovement!(execute, self:CCP Anim Instance Galatic New Object Reference, XMovement:float=@n4, YMovement:float=@n4.YMovement) -> (then)
   Event BlueprintInitializeAnimation()
   > Set Skel_Mesh(Skel_Mesh=GetOwnerSkeletalMeshComponent())
   n5 = Branch(Condition=Get bIsSkyDiving())
   > n6 = Branch(Condition="float > float"(A=Get PawnVelocityZ(), B=50) {is=title K2Node_CallFunction})
   > Set BwdMovement_0(BwdMovement_0:double=float * float(A=Get BwdMovement_0() -> (BwdMovement_0:double), B=0)) -> (Output_Get:double)
   Set BwdMovement_0(execute=[@n5.else, @n6.else], BwdMovement_0=float * float(A=Get BwdMovement_0(), B=1))
   n7 = Branch(Condition="Equal Exactly (String)"(A=Get weapon() -> (weapon:String), B=Pistol))
   > Set bHair_0(bHair_0:bool=true) -> (Output_Get:bool)
   Set bHair_0(execute=@n7.else, bHair_0=false)
   n8 = Get Skel_Mesh()
   Set ThighRotL_0(ThighRotL_0:double=GetAlphaFromReferenceTransformDelta(SkeletalMeshComponent=@n8, BoneName=thigh_l, BaseName=pelvis, InputRange=(X=-20,Y=-100), TwistAxis="1,0,0")) -> (Output_Get:double)
   > Set ThighRotR_0(ThighRotR_0:double=GetAlphaFromReferenceTransformDelta(SkeletalMeshComponent=@n8, BoneName=thigh_r, BaseName=pelvis, InputRange=(X=-20,Y=-100), TwistAxis="1,0,0")) -> (Output_Get:double)
   > n9 = Cast To FortPlayerAnimInstance(Object=GetAnimInstance(self=GetOwnerSkeletalMeshComponent()))
   > Set weapon(weapon:String=Enum to String!(Enumerator:EFortWeaponCoreAnimation Enum=Get WeaponCoreAnim(self=@n9)) -> (ReturnValue:String)) -> (Output_Get:String)
   n10 = GetOwnerSkeletalMeshComponent()
   n11 = BreakRotator(InRot=BreakTransform(InTransform=GetDeltaTransformFromRefPose(self=@n10, BoneName=thigh_r, BaseName=pelvis)).Rotation)
   > Set ThighRotUpR_0(ThighRotUpR_0:double=MapRangeClamped(Value=@n11.Yaw, InRangeA=-15, InRangeB=-100)) -> (Output_Get:double)
   > Set ThighRotInrR_0(ThighRotInrR_0:double=MapRangeClamped(Value=@n11.Pitch, InRangeA=0, InRangeB=-10, OutRangeA=1, OutRangeB=0)) -> (Output_Get:double)
   > Set ThighRotOtrR_0(ThighRotOtrR_0:double=MapRangeClamped(Value=@n11.Pitch, InRangeA=20, InRangeB=50)) -> (Output_Get:double)
   n12 = GetOwnerSkeletalMeshComponent()
   n13 = BreakRotator(InRot=BreakTransform(InTransform=GetDeltaTransformFromRefPose(self=@n12, BoneName=thigh_l, BaseName=pelvis)).Rotation)
   > Set ThighRotUpL_0(ThighRotUpL_0:double=MapRangeClamped(Value=@n13.Yaw, InRangeA=-10, InRangeB=-100)) -> (Output_Get:double)
   > Set ThighRotInrL_0(ThighRotInrL_0:double=MapRangeClamped(Value=@n13.Pitch, InRangeA=0, InRangeB=-10, OutRangeA=1, OutRangeB=0)) -> (Output_Get:double)
   > Set ThighRotOtrL_0(ThighRotOtrL_0:double=MapRangeClamped(Value=@n13.Pitch, InRangeA=20, InRangeB=50)) -> (Output_Get:double)
   > Set jumpingAlpha_0(jumpingAlpha_0:double="To Float (Boolean)"(InBool=Get bIsJumping())) -> (Output_Get:double)
   n14 = Get bIsMoving2D()
   > Set isNotCrouchRunning_0(isNotCrouchRunning_0:double="To Float (Boolean)"(InBool=NOT Boolean(A=@n14))) -> (Output_Get:double)
   > Set ParachuteAlpha_0(ParachuteAlpha_0:double="To Float (Boolean)"(InBool=Get bIsParachuteOpened())) -> (Output_Get:double)
   > Set skydiveAlpha_0(skydiveAlpha_0:double="To Float (Boolean)"(InBool=Get bIsSkyDiving())) -> (Output_Get:double)
   n15 = MapRangeClamped(Value=@n4.YMovement, InRangeA=0, InRangeB=-1)
   n16 = float * float(A="Absolute (Float)"(A=@n4), B=0.5)
   Set BwdMovement_0(BwdMovement_0=MapRangeClamped(
     Value=float + float(A=@n16, B=@n15) {is=title K2Node_CommutativeAssociativeBinaryOperator}
     InRangeA=0
     InRangeB=1.1
     OutRangeA=0.4
     OutRangeB=0
    ))
   n17 = MapRangeClamped(Value=@n4, InRangeA=-0.2, InRangeB=1)
   n18 = float * float(A="Absolute (Float)"(A=@n4.YMovement), B=0.2)
   > Set RightMovement_0(RightMovement_0:double=MapRangeClamped(Value=float + float(A=@n18, B=@n17), InRangeA=0, InRangeB=1.2, OutRangeB=0.5)) -> (Output_Get:double)
   n19 = MapRangeClamped(Value=@n4, InRangeA=0.2, InRangeB=-1, OutRangeB=0.3)
   n20 = float * float(A="Absolute (Float)"(A=@n4.YMovement), B=0.2)
   > Set LeftMovement_0(LeftMovement_0:double=MapRangeClamped(Value=float + float(A=@n20, B=@n19), InRangeA=0, InRangeB=1.2, OutRangeB=0.5)) -> (Output_Get:double)
   Get bIsCrouchSprinting()
   Get bHair_0() -> (bHair_0:bool)
  ]}
]
variables [
 bHair_0 {type=boolean}
 BwdMovement_0 {type=number}
 finalGravity {type=Vector}
 Gravity Override Params {type=GravityOverrideParamsStruct}
 GravityOverride {type=Vector}
 InComponentSpace {type=boolean}
 InComponentSpaceLastFrame {type=boolean}
 InterpSpeed {type=number}
 isNotCrouchRunning_0 {type=number}
 jumpingAlpha_0 {type=number}
 LeftMovement_0 {type=number}
 LockedOnSimSpace {type=boolean}
 ParachuteAlpha_0 {type=number}
 RightMovement_0 {type=number}
 sim_space_rotation {type=Rotator}
 simSpaceRoot {type=number}
 simSpaceSpine {type=number}
 Skel_Mesh {type=SkeletalMeshComponent}
 skydiveAlpha_0 {type=number}
 ThighRotInrL_0 {type=number}
 ThighRotInrR_0 {type=number}
 ThighRotL_0 {type=number}
 ThighRotOtrL_0 {type=number}
 ThighRotOtrR_0 {type=number}
 ThighRotR_0 {type=number}
 ThighRotUpL_0 {type=number}
 ThighRotUpR_0 {type=number}
 weapon {type=string, default=""}
 windVector {type=Vector}
]
