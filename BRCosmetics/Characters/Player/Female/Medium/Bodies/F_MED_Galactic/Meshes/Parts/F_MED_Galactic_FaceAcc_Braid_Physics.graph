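# Physics asset for the braid on the F_MED_Galactic female character's FaceAcc part, providing bodies for physics-driven hair motion.
# Four capsules form a chain along dyn_hair_braid_02 to 05, with angular limits between segments that keep twist tighter than swing.
# Two larger capsules cover spine_04 and spine_05 as upper-torso volumes.
# The braid root attaches to spine_05 with free swing and twist, as does spine_05 to spine_04.
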PhysicsAsset /BRCosmetics/Characters/Player/Female/Medium/Bodies/F_MED_Galactic/Meshes/Parts/F_MED_Galactic_FaceAcc_Braid_Physics
bodies {
 dyn_hair_braid_02 [{
   shapeType=Capsule
   length=6.702
   radius=3.1249
   shapeName=dyn_hair_braid_02_capsule
   center=<4.7781 -0.377 0.0895>
   rotation=<rot 0 87.4801 -85.377>
  }]
 dyn_hair_braid_03 [{shapeType=Capsule, length=3.7841, radius=2.2482, shapeName=dyn_hair_braid_03_capsule, center=<4.0452 0 0>, rotation=<rot 90 0 0>}]
 dyn_hair_braid_04 [{shapeType=Capsule, length=1.2486, radius=2.25, shapeName=dyn_hair_braid_04_capsule, center=<2.8772 0 0>, rotation=<rot 90 0 0>}]
 dyn_hair_braid_05 [{shapeType=Capsule, length=3.0277, radius=2.5, shapeName=dyn_hair_braid_05_capsule, center=<4.5 0 0>, rotation=<rot 90 0 0>}]
 spine_04 [{shapeType=Capsule, length=14.3937, radius=11.7562, shapeName=spine_04_capsule, center=<0.8279 -0.8816 0>, rotation=<rot 0 -89.9999 0>}]
 spine_05 [{shapeType=Capsule, length=16.9347, radius=9.9356, shapeName=spine_05_capsule, center=<2.1215 -2.9319 0>, rotation=<rot 0 -8.0783 0>}]
}
constraints [
 {
  bone1Name=spine_05
  bone2Name=spine_04
  swing1LimitDegrees=45
  swing1Motion=Free
  swing2LimitDegrees=45
  swing2Motion=Free
  twistLimitDegrees=45
  twistMotion=Free
 }
 {bone1Name=dyn_hair_braid_03, bone2Name=dyn_hair_braid_02, swing1LimitDegrees=45, swing2LimitDegrees=45, twistLimitDegrees=15}
 {bone1Name=dyn_hair_braid_04, bone2Name=dyn_hair_braid_03, swing1LimitDegrees=45, swing2LimitDegrees=45, twistLimitDegrees=15}
 {bone1Name=dyn_hair_braid_05, bone2Name=dyn_hair_braid_04, swing1LimitDegrees=45, swing2LimitDegrees=45, twistLimitDegrees=15}
 {
  bone1Name=dyn_hair_braid_02
  bone2Name=spine_05
  swing1LimitDegrees=45
  swing1Motion=Free
  swing2LimitDegrees=45
  swing2Motion=Free
  twistLimitDegrees=45
  twistMotion=Free
 }
]
