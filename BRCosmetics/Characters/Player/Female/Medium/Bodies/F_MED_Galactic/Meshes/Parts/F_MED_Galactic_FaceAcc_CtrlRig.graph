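# Control Rig for the Galactic female face accessory mesh that adjusts the hair braid according to the right arm's pose.
# It converts upperarm_r's rotation relative to spine_05 to Euler and remaps absolute Z to a clamped weight.
# That weight drives a Modify Transforms node applying a fixed Z rotation to dyn_hair_braid_02.
# A second, similar chain computes a Y rotation from the same arm angle, but its output is unconnected.
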ControlRigBlueprint /BRCosmetics/Characters/Player/Female/Medium/Bodies/F_MED_Galactic/Meshes/Parts/F_MED_Galactic_FaceAcc_CtrlRig
graphs [
 RigVMFunctionLibrary {links [], nodes []}
 RigVMModel {nodes [
   RigUnit_BeginExecution = &RigUnit_BeginExecution()
   n1 = &RigVMFunction_MathQuaternionToEuler(
    RotationOrder=ZYX
    Value=&RigUnit_GetRelativeTransformForItem(Child=(Type=Bone,Name="upperarm_r"), Parent=(Type=Bone,Name="spine_05")).RelativeTransform.Rotation
   )
   &RigUnit_ModifyTransforms(
    ItemToModify=((Item=(Type=Bone),Transform=(Rotation=(X=0,Y=0,Z=0,W=1),Translation=(X=0,Y=0,Z=0),Scale3D=(X=1,Y=1,Z=1))),(Item=(Type=Bone,Name="dyn_hair_braid_02"),Transform=(Rotation=(X=0,Y=0,Z=0,W=1),Translation=(X=0,Y=0,Z=0),Scale3D=(X=1,Y=1,Z=1))))
    Weight=&RigVMFunction_MathFloatRemap(
     SourceMaximum=75
     SourceMinimum=30
     TargetMinimum=0
     bClamp=true
     Value=/Script/RigVM.RigVMFunction_MathFloatAbs(Value=@n1.Result.Z) {is=struct RigVMUnitNode, template="Absolute::Execute(in Value,out Result)"}.Result
    ).Result
    ItemToModify.1.Transform.Rotation=&RigVMFunction_MathQuaternionFromEuler(Euler=(X=0,Y=0,Z=-20)).Result
    ExecutePin=@RigUnit_BeginExecution.ExecutePin
   )
   n2 = &RigVMFunction_MathQuaternionToEuler(
    RotationOrder=ZYX
    Value=&RigUnit_GetRelativeTransformForItem(Child=(Type=Bone,Name="upperarm_r"), Parent=(Type=Bone,Name="spine_05")).RelativeTransform.Rotation
   )
   &RigVMFunction_MathQuaternionFromEuler(
    Euler=(X=0,Y=0,Z=0)
    Euler.Y=&RigVMFunction_MathFloatRemap(
     SourceMaximum=122
     SourceMinimum=50
     TargetMaximum=-70
     TargetMinimum=0
     bClamp=true
     Value=&RigVMFunction_MathFloatAbs(Value=@n2.Result.Z) {template="Absolute::Execute(in Value,out Result)"}.Result
    ).Result
   )
  ]}
]
hierarchy [
 root {loc=<0 0 0>, rot=<0 0 0 1>}
 pelvis {parent=root, loc=<0 -0.5301 98.0571>, rot=<0.0127 0.707 -0.0127 -0.707>}
 spine_01 {parent=pelvis, loc=<6.1208 -0.9328 0>, rot=<0 0 0.034 -0.9994>}
 spine_02 {parent=spine_01, loc=<6.5996 -0.2095 0>, rot=<0 0 0.0238 -0.9997>}
 spine_03 {parent=spine_02, loc=<6.4914 -0.1237 0>, rot=<0 0 0.0149 -0.9999>}
 spine_04 {parent=spine_03, loc=<6.9598 0.0794 0>, rot=<0 0 -0.0132 -0.9999>}
 spine_05 {parent=spine_04, loc=<6.9987 0 0>, rot=<0 0 -0.0015 -1>}
 clavicle_l {parent=spine_05, loc=<5.4831 -0.1053 -3.2946>, rot=<0.0465 -0.7289 -0.009 -0.6829>}
 upperarm_l {parent=clavicle_l, loc=<9.8658 -1.2136 -0.4105>, rot=<0.0164 -0.3245 -0.0269 -0.9454>}
 upperarm_twist_01_l {parent=upperarm_l, loc=<7.2135 -0.1134 0>, rot=<0 0 0 1>}
 clavicle_r {parent=spine_05, loc=<5.4831 -0.1053 3.2946>, rot=<0.7289 0.0465 -0.6829 0.009>}
 upperarm_r {parent=clavicle_r, loc=<-9.8658 1.2136 0.4104>, rot=<0.0164 -0.3245 -0.0269 -0.9454>}
 upperarm_twist_01_r {parent=upperarm_r, loc=<-7.2135 0.1134 0>, rot=<0 0 0 1>}
 dyn_con_hair_tail_10 {parent=clavicle_r, loc=<-4.0788 -0.5451 -8.0146>, rot=<0.5562 -0.7582 0.1005 0.325>}
 neck_01 {parent=spine_05, loc=<11.3138 0.1529 0>, rot=<0 0 -0.2003 -0.9797>}
 neck_02 {parent=neck_01, loc=<4.5627 0 0>, rot=<0 0 0.0504 -0.9987>}
 head {parent=neck_02, loc=<5.3914 0 0>, rot=<0 0 0.115 -0.9934>}
 faceAttach {parent=head, loc=<0.6201 0.9296 0>, rot=<0.0082 -0.7071 -0.0082 -0.7071>, scale=<0.89 0.89 0.89>}
 C_brow_mid {parent=faceAttach, loc=<-0.0018 9.2931 3.5986>, rot=<0 -0.0528 0.0104 -0.9986>}
 L_brow_mid {parent=faceAttach, loc=<3.8334 8.2993 5.0089>, rot=<0 -0.0036 0.2553 -0.9669>}
 L_brow_outer {parent=faceAttach, loc=<5.7701 6.1251 4.9388>, rot=<0.2832 0.3529 0.3264 -0.8299>}
 L_cheek_inner {parent=faceAttach, loc=<4.166 7.3352 -1.5946>, rot=<-0.0526 -0.2243 0.3244 -0.9175>}
 C_lip_upper_mid {parent=faceAttach, loc=<-0.0031 9.0988 -3.587>, rot=<0.2225 0.0007 -0.0002 -0.9749>}
 L_lip_upper_outer {parent=faceAttach, loc=<1.259 8.4924 -3.6878>, rot=<0.0163 -0.1277 0.2829 -0.9505>}
 L_lip_corner {parent=faceAttach, loc=<2.4684 7.6489 -3.9392>, rot=<0.0035 0.0074 -0.0147 -0.9999>}
 C_nose_bridge {parent=faceAttach, loc=<0.0045 10.1501 0.2257>, rot=<-0.2218 0.0111 0.0031 -0.975>}
 R_brow_mid {parent=faceAttach, loc=<-3.8314 8.2986 5.0094>, rot=<-0.0001 0.0042 -0.2553 -0.9669>}
 R_brow_outer {parent=faceAttach, loc=<-5.7676 6.1248 4.9388>, rot=<0.036 -0.0414 -0.4337 -0.8994>}
 R_cheek_inner {parent=faceAttach, loc=<-4.1645 7.3352 -1.5945>, rot=<-0.0535 0.2264 -0.3241 -0.917>}
 R_lip_corner {parent=faceAttach, loc=<-2.4591 7.6462 -3.9375>, rot=<0.0032 -0.0078 0.0146 -0.9999>}
 C_jaw {parent=faceAttach, loc=<-0.2249 1.0714 -3.3043>, rot=<0 0 0 1>}
 L_lip_lower_outer {parent=C_jaw, loc=<1.5078 7.3123 -0.3743>, rot=<0.0175 0.0689 0.324 -0.9434>}
 C_lip_lower_mid {parent=C_jaw, loc=<0.2186 7.7389 -0.6832>, rot=<0 -0.0508 -0.0108 -0.9987>}
 R_lip_lower_outer {parent=C_jaw, loc=<-1.0153 7.2919 -0.5306>, rot=<0.0076 -0.0155 -0.2835 -0.9588>}
 R_eye {parent=faceAttach, loc=<-3.2214 6.2373 2.8822>, rot=<0 0 0.9998 -0.0187>}
 R_eye_lid_upper_mid {parent=R_eye, loc=<-0.0396 -1.1714 0.2631>, rot=<0.0523 -0.0009 0.0167 -0.9985>}
 R_eye_lid_lower_mid {parent=R_eye, loc=<-0.0085 -0.9425 -0.7051>, rot=<0.0329 -0.0005 0.0167 -0.9993>}
 L_eye {parent=faceAttach, loc=<3.1563 6.2373 2.8822>, rot=<0 0 0.9998 0.0187>}
 L_eye_lid_upper_mid {parent=L_eye, loc=<0.0251 -1.1623 0.2388>, rot=<0.0523 0 0 -0.9986>}
 L_eye_lid_lower_mid {parent=L_eye, loc=<0.0223 -0.923 -0.7488>, rot=<0.0329 0 0 -0.9995>}
 R_lip_upper_outer {parent=faceAttach, loc=<-1.2579 8.4918 -3.6876>, rot=<0.0139 0.1345 -0.2853 -0.9488>}
 dyn_hair_braid_MSTR {parent=head, loc=<-0.909 -3.7774 4.7607>, rot=<0.1115 -0.1313 0.9847 -0.0256>}
 dyn_hair_braid_01 {parent=dyn_hair_braid_MSTR, loc=<2.1065 0 0>, rot=<0.1591 0.0551 0.2395 -0.9562>}
 dyn_hair_braid_02 {parent=dyn_hair_braid_01, loc=<5.7222 0 0>, rot=<-0.6001 -0.0702 0.2846 -0.7443>}
 dyn_hair_braid_03 {parent=dyn_hair_braid_02, loc=<10.3444 0 0>, rot=<-0.141 -0.1848 0.0128 -0.9725>}
 dyn_hair_braid_04 {parent=dyn_hair_braid_03, loc=<8.5721 0 0>, rot=<0.0456 -0.1361 -0.0231 -0.9894>}
 dyn_hair_braid_05 {parent=dyn_hair_braid_04, loc=<5.0559 0 0>, rot=<0.0379 -0.0598 -0.0723 -0.9949>}
 dyn_con_hair_tail_11 {parent=neck_02, loc=<1.6083 -3.1647 5.2091>, rot=<0.1658 -0.275 0.889 0.3263>}
 dyn_con_hair_tail_09 {parent=spine_05, loc=<6.1793 8.2619 9.6619>, rot=<0.1524 0.6041 0.7612 0.1802>}
 dyn_con_hair_tail_08 {parent=spine_04, loc=<5.5247 12.1878 9.785>, rot=<0.0521 0.584 0.8057 0.0842>}
]
