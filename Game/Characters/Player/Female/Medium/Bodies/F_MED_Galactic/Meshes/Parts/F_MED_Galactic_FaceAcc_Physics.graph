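# Physics asset for the face accessory part of the Galactic medium female player body, covering a ponytail bone, the head and the upper spine.
# It has three capsule bodies: a small ponytail capsule, a tiny head capsule and a large, nearly spherical, offset capsule on spine_05.
# Both constraints pair with spine_05: the head rotates freely, while the ponytail links directly to the spine, bypassing the head, with uneven swing limits.
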PhysicsAsset /Game/Characters/Player/Female/Medium/Bodies/F_MED_Galactic/Meshes/Parts/F_MED_Galactic_FaceAcc_Physics
bodies {
 dyn_ponytail_01 [{
   shapeType=Capsule
   length=4.1737
   radius=3.206
   shapeName=dyn_ponytail_01_capsule
   center=<3.4798 -0.3421 -0.0365>
   rotation=<rot 0 -91.106 -76.1549>
  }]
 head [{shapeType=Capsule, length=1, radius=1}]
 spine_05 [{shapeType=Capsule, length=0.505, radius=24.755, shapeName=spine_05_capsule, center=<-10.7133 -13.2153 0>, rotation=<rot 0 -90 0>}]
}
constraints [
 {
  bone1Name=head
  bone2Name=spine_05
  swing1LimitDegrees=45
  swing1Motion=Free
  swing2LimitDegrees=45
  swing2Motion=Free
  twistLimitDegrees=45
  twistMotion=Free
 }
 {bone1Name=dyn_ponytail_01, bone2Name=spine_05, swing1LimitDegrees=16.4286, swing2LimitDegrees=32, twistLimitDegrees=45}
]
