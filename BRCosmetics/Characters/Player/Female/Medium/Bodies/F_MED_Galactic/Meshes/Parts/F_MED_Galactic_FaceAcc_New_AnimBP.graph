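# Post-body animation Blueprint for the F_MED_Galactic character's face accessory, driving its hair braid.
# Copied body poses feed weighted constraints placing braid bones between head, neck, spine and clavicle, then a head-twist-driven offset.
# A Control Rig engages for Pistol weapon animations before the cached pose reaches rigid-body simulation, which uses a separate physics asset for front end or emotes.
# AnimDynamics, ponytail constraint and other leftover chains are disconnected from the output.
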AnimBlueprint /BRCosmetics/Characters/Player/Female/Medium/Bodies/F_MED_Galactic/Meshes/Parts/F_MED_Galactic_FaceAcc_New_AnimBP
target_skeleton=&/F_MED_Galactic_FaceAcc_Skeleton
graphs [
 AnimGraph {nodes [
   n1 = Anim Dynamics - Bone: dyn_hair_braid_04(Alpha=1, ExternalForce="0,0,0") {
    angularDampingOverride=1
    angularSpringConstant=1500
    bAngularSpring=true
    bGravityOverrideInSimSpace=true
    bOverrideAngularDamping=true
    gravityScale=0.35
    simulationSpace=RootRelative
    alphaBoolBlend {blendOption=Linear}
    boundBone {boneName=dyn_hair_braid_04}
    componentLinearAccScale=<0.3 0.3 0.3>
    componentLinearVelScale=<0.3 0.3 0.3>
    physicsBodyDefinitions [{
      sphereCollisionRadius=0
      boundBone {boneName=dyn_hair_braid_04}
      boxExtents=<10 20 20>
      constraintSetup {angularLimitsMax=<1 5 10>, angularLimitsMin=<-1 -15 -10>, angularTarget=<0 0 0>}
      localJointOffset=<5 0 1>
     }]
    relativeSpaceBone {boneName=root}
   }
   n2 = Anim Dynamics - Bone: dyn_hair_braid_04(ComponentPose=@n1, Alpha=1, ExternalForce="0,0,0") {
    angularDampingOverride=1
    angularSpringConstant=1500
    bAngularSpring=true
    bGravityOverrideInSimSpace=true
    bOverrideAngularDamping=true
    gravityScale=0.35
    simulationSpace=RootRelative
    alphaBoolBlend {blendOption=Linear}
    boundBone {boneName=dyn_hair_braid_04}
    componentLinearAccScale=<0.3 0.3 0.3>
    componentLinearVelScale=<0.3 0.3 0.3>
    physicsBodyDefinitions [{
      sphereCollisionRadius=0
      boundBone {boneName=dyn_hair_braid_04}
      boxExtents=<10 20 20>
      constraintSetup {angularLimitsMax=<1 5 10>, angularLimitsMin=<-1 -15 -10>, angularTarget=<0 0 0>}
      localJointOffset=<5 0 1>
     }]
    relativeSpaceBone {boneName=root}
   }
   Anim Dynamics - Bone: dyn_hair_braid_05(ComponentPose=@n2, Alpha=1, ExternalForce="0,0,0") {
    angularDampingOverride=1
    angularSpringConstant=1500
    bAngularSpring=true
    bGravityOverrideInSimSpace=true
    bOverrideAngularDamping=true
    gravityScale=0.35
    simulationSpace=RootRelative
    alphaBoolBlend {blendOption=Linear}
    boundBone {boneName=dyn_hair_braid_05}
    physicsBodyDefinitions [{
      sphereCollisionRadius=0
      boundBone {boneName=dyn_hair_braid_05}
      boxExtents=<10 20 20>
      constraintSetup {angularLimitsMax=<1 5 15>, angularLimitsMin=<-1 -15 -15>, angularTarget=<0 0 0>}
      localJointOffset=<5 0 2>
     }]
   }
   n3 = RigidBody(ComponentPose=Local To Component()) {
    overridePhysicsAsset=/Game/Characters/Player/Female/Medium/Bodies/F_MED_Galactic/Meshes/Parts/F_MED_Galactic_FaceAcc_Physics
    alphaBoolBlend {blendOption=Linear}
   }
   n4 = Anim Dynamics - Bone: dyn_ponytail_01(ComponentPose=@n3, Alpha=0.45, ExternalForce="0,0,0") {
    angularDampingOverride=1
    angularSpringConstant=1500
    bAngularSpring=true
    bGravityOverrideInSimSpace=true
    bOverrideAngularDamping=true
    gravityScale=0.35
    simulationSpace=RootRelative
    alphaBoolBlend {blendOption=Linear}
    alphaScaleBiasClamp {bInterpResult=true}
    boundBone {boneName=dyn_ponytail_01}
    componentAppliedLinearAccClamp=<1100 1100 1100>
    componentLinearAccScale=<0.7 0.7 0.4>
    componentLinearVelScale=<0.7 0.7 0.4>
    physicsBodyDefinitions [{
      sphereCollisionRadius=0
      boundBone {boneName=dyn_ponytail_01}
      boxExtents=<10 20 20>
      constraintSetup {angularLimitsMax=<0 60 0>, angularLimitsMin=<0 -10 -10>, angularTarget=<0 0 0>}
      localJointOffset=<7 0 -1>
     }]
   }
   n5 = Anim Dynamics - Bone: dyn_ponytail_02(ComponentPose=@n4, Alpha=1, ExternalForce="0,0,0") {
    angularDampingOverride=1
    angularSpringConstant=1500
    bAngularSpring=true
    bGravityOverrideInSimSpace=true
    bOverrideAngularDamping=true
    gravityScale=0.35
    simulationSpace=RootRelative
    alphaBoolBlend {blendOption=Linear}
    alphaScaleBiasClamp {bInterpResult=true}
    boundBone {boneName=dyn_ponytail_02}
    componentAppliedLinearAccClamp=<1100 1100 1100>
    componentLinearAccScale=<0.7 0.7 0.4>
    componentLinearVelScale=<0.7 0.7 0.4>
    physicsBodyDefinitions [{
      sphereCollisionRadius=0
      boundBone {boneName=dyn_ponytail_02}
      boxExtents=<10 20 20>
      constraintSetup {angularLimitsMax=<0 65 0>, angularLimitsMin=<0 -10 -10>, angularTarget=<0 0 0>}
      localJointOffset=<5 0 2>
     }]
   }
   n6 = Anim Dynamics - Bone: dyn_hair_04(ComponentPose=@n5, Alpha=0.65, ExternalForce="0,0,0") {
    angularDampingOverride=1
    angularSpringConstant=1500
    bAngularSpring=true
    bGravityOverrideInSimSpace=true
    bOverrideAngularDamping=true
    gravityScale=0.35
    simulationSpace=RootRelative
    alphaBoolBlend {blendOption=Linear}
    boundBone {boneName=dyn_hair_04}
    componentAppliedLinearAccClamp=<1100 1100 1100>
    componentLinearAccScale=<0.7 0.7 0.4>
    componentLinearVelScale=<0.7 0.7 0.4>
    physicsBodyDefinitions [{
      sphereCollisionRadius=0
      boundBone {boneName=dyn_hair_04}
      boxExtents=<8 15 15>
      constraintSetup {angularLimitsMax=<0 55 0>, angularLimitsMin=<0 -5 -40>, angularTarget=<0 0 0>}
      localJointOffset=<6 0 0>
     }]
   }
   Anim Dynamics - Bone: dyn_hair_05(ComponentPose=@n6, Alpha=0.85, ExternalForce="0,0,0") {
    angularDampingOverride=1
    angularSpringConstant=1500
    bAngularSpring=true
    bGravityOverrideInSimSpace=true
    bOverrideAngularDamping=true
    gravityScale=0.31
    simulationSpace=RootRelative
    alphaBoolBlend {blendOption=Linear}
    boundBone {boneName=dyn_hair_05}
    componentAppliedLinearAccClamp=<1100 1100 1100>
    componentLinearAccScale=<0.7 0.7 0.4>
    componentLinearVelScale=<0.7 0.7 0.4>
    physicsBodyDefinitions [{
      sphereCollisionRadius=0
      boundBone {boneName=dyn_hair_05}
      boxExtents=<5 15 15>
      constraintSetup {angularLimitsMax=<0 40 0>, angularLimitsMin=<0 -10 -40>, angularTarget=<0 0 0>}
      localJointOffset=<3 0 0>
     }]
   }
   n7 = "Constraint - dyn_parent_ponytail (P)"(Alpha=1, ConstraintWeights_0=1) {
    alphaBoolBlend {blendOption=Linear}
    boneToModify {boneName=dyn_parent_ponytail}
    constraintSetup [{offsetOption=None, targetBone {boneName=dyn_con_hair_tail_01}}]
    constraintWeights [1]
   }
   n8 = "Constraint - dyn_ponytail_MSTR (P)"(ComponentPose=@n7, Alpha=1, ConstraintWeights_0=1) {
    alphaBoolBlend {blendOption=Linear}
    boneToModify {boneName=dyn_ponytail_MSTR}
    constraintSetup [{offsetOption=None, targetBone {boneName=dyn_con_hair_tail_02}}]
    constraintWeights [1]
   }
   "Constraint - dyn_ponytail_01 (P)"(ComponentPose=@n8, Alpha=1, ConstraintWeights_0=1) {
    alphaBoolBlend {blendOption=Linear}
    boneToModify {boneName=dyn_ponytail_01}
    constraintSetup [{offsetOption=None, targetBone {boneName=dyn_con_hair_tail_03}}]
    constraintWeights [1]
   }
   "Transform (Modify) Bone - Bone: dyn_hair_braid_03"(Alpha=1, Translation="0.000,-7,0.000", Rotation="0,0,0", Scale="1,1,1") {
    rotationMode=BMM_Additive
    rotationSpace=BCS_BoneSpace
    translationMode=BMM_Additive
    translationSpace=BCS_BoneSpace
    alphaBoolBlend {blendOption=Linear}
    boneToModify {boneName=dyn_hair_braid_03}
    translation {y=-7}
   }
   n9 = RigidBody!(ComponentPose=Local To Component(LocalPose=Use cached pose 'prerigid'()), Alpha, OverrideWorldGravity:Vector="0,0,-980", ExternalForce) -> (Pose) {
    bClampLinearTranslationLimitToRefPose=true
    bOverrideWorldGravity=true
    overridePhysicsAsset=&/F_MED_Galactic_FaceAcc_Braid_Physics
    simulationSpace=BaseBoneSpace
    alphaBoolBlend {blendOption=Linear}
    baseBoneRef {boneName=spine_04}
    componentAppliedLinearAccClamp=<300 300 300>
    componentLinearAccScale=<0.2 0.2 0.3>
    componentLinearVelScale=<0.1 0.1 0.2>
   }
   n10 = Component To Local(ComponentPose=@n9)
   n11 = RigidBody(ComponentPose=Local To Component(LocalPose=Use cached pose 'prerigid'()), OverrideWorldGravity="0,0,-980") {
    bClampLinearTranslationLimitToRefPose=true
    bOverrideWorldGravity=true
    overridePhysicsAsset=&/F_MED_Galactic_FaceAcc_Braid_FrontEnd_Physics
    simulationSpace=BaseBoneSpace
    alphaBoolBlend {blendOption=Linear}
    baseBoneRef {boneName=spine_04}
    componentAppliedLinearAccClamp=<300 300 300>
    componentLinearAccScale=<0.2 0.2 0.3>
    componentLinearVelScale=<0.1 0.1 0.2>
   }
   n12 = Component To Local(ComponentPose=@n11)
   n13 = Get bIsPlayingEmote()
   n14 = RigidBody(ComponentPose=Local To Component(LocalPose=Use cached pose 'prerigid'()), OverrideWorldGravity="0,0,-980") {
    bClampLinearTranslationLimitToRefPose=true
    bOverrideWorldGravity=true
    overridePhysicsAsset=&F_MED_Galactic_FaceAcc_Braid_FrontEnd_Physics
    simulationSpace=BaseBoneSpace
    alphaBoolBlend {blendOption=Linear}
    baseBoneRef {boneName=spine_04}
    componentAppliedLinearAccClamp=<300 300 300>
    componentLinearAccScale=<0.2 0.2 0.3>
    componentLinearVelScale=<0.1 0.1 0.2>
   }
   Output Pose(Result=Blend Poses by bool(
     BlendPose_0=Component To Local(ComponentPose=@n14)
     BlendPose_1=Blend Poses by bool(BlendPose_0=@n12, BlendPose_1=@n10, BlendTime_1=1, bActiveValue=@n13) {blendType=Linear}
     bActiveValue=Get bIsInFrontEnd()
    ) {blendType=Linear}) {layerGroup=None}
   n15 = Copy Pose From Mesh(SourceMeshComponent=Get Body_Mesh())
   n16 = Copy Pose From Mesh() {bCopyCurves=true, bUseAttachedParent=true}
   n17 = "Constraint - dyn_con_hair_tail_11 (PPP)"!(
    ComponentPose=Local To Component(LocalPose=Layered blend per bone(BasePose=@n16, BlendPoses_0=@n15) {layerSetup [{branchFilters [{boneName=upperarm_l}, {boneName=upperarm_r}]}]})
    Alpha=1
    ConstraintWeights_0=0.8008
    ConstraintWeights_1:float=0.1778
    ConstraintWeights_2:float=0.0214
   ) -> (Pose) {
    alphaBoolBlend {blendOption=Linear}
    boneToModify {boneName=dyn_con_hair_tail_11}
    constraintSetup [{targetBone {boneName=head}}, {targetBone {boneName=neck_02}}, {targetBone {boneName=neck_01}}]
    constraintWeights [1,1,1]
   }
   n18 = "Constraint - dyn_con_hair_tail_10 (PPPP)"!(
    ComponentPose=@n17
    Alpha=1
    ConstraintWeights_0=0.33
    ConstraintWeights_1:float=0.27
    ConstraintWeights_2:float=0.3
    ConstraintWeights_3:float=0.1
   ) -> (Pose) {
    alphaBoolBlend {blendOption=Linear}
    boneToModify {boneName=dyn_con_hair_tail_10}
    constraintSetup [{targetBone {boneName=spine_05}}, {targetBone {boneName=clavicle_r}}, {targetBone {boneName=neck_02}}, {targetBone {boneName=head}}]
    constraintWeights [1,1,1,1]
   }
   n19 = "Constraint - dyn_con_hair_tail_09 (PPP)"!(ComponentPose=@n18, Alpha=1, ConstraintWeights_0=0.6591, ConstraintWeights_1:float=0.2295, ConstraintWeights_2:float=0.1114) -> (Pose) {
    alphaBoolBlend {blendOption=Linear}
    boneToModify {boneName=dyn_con_hair_tail_09}
    constraintSetup [{targetBone {boneName=spine_05}}, {targetBone {boneName=spine_04}}, {targetBone {boneName=clavicle_r}}]
    constraintWeights [1,1,1]
   }
   n20 = "Constraint - dyn_con_hair_tail_08 (PP)"!(ComponentPose=@n19, Alpha=1, ConstraintWeights_0=0.805, ConstraintWeights_1:float=0.195) -> (Pose) {
    alphaBoolBlend {blendOption=Linear}
    boneToModify {boneName=dyn_con_hair_tail_08}
    constraintSetup [{targetBone {boneName=spine_04}}, {targetBone {boneName=spine_05}}]
    constraintWeights [1,1]
   }
   n21 = "Constraint - dyn_hair_braid_01 (P)"(ComponentPose=@n20, Alpha=1, ConstraintWeights_0=1) {
    alphaBoolBlend {blendOption=Linear}
    boneToModify {boneName=dyn_hair_braid_01}
    constraintSetup [{offsetOption=None, targetBone {boneName=dyn_con_hair_tail_11}}]
    constraintWeights [1]
   }
   n22 = "Constraint - dyn_hair_braid_02 (P)"(ComponentPose=@n21, Alpha=1, ConstraintWeights_0=1) {
    alphaBoolBlend {blendOption=Linear}
    boneToModify {boneName=dyn_hair_braid_02}
    constraintSetup [{offsetOption=None, targetBone {boneName=dyn_con_hair_tail_10}}]
    constraintWeights [1]
   }
   n23 = "Constraint - dyn_hair_braid_03 (P)"(ComponentPose=@n22, Alpha=1, ConstraintWeights_0=1) {
    alphaBoolBlend {blendOption=Linear}
    boneToModify {boneName=dyn_hair_braid_03}
    constraintSetup [{offsetOption=None, targetBone {boneName=dyn_con_hair_tail_09}}]
    constraintWeights [1]
   }
   n24 = "Constraint - dyn_hair_braid_04 (P)"(ComponentPose=@n23, Alpha=1, ConstraintWeights_0=1) {
    alphaBoolBlend {blendOption=Linear}
    boneToModify {boneName=dyn_hair_braid_04}
    constraintSetup [{offsetOption=None, targetBone {boneName=dyn_con_hair_tail_08}}]
    constraintWeights [1]
   }
   n25 = "Transform (Modify) Bone - Bone: dyn_hair_braid_MSTR"(ComponentPose=@n24, Alpha=Get head delta() -> (head delta:double), Translation="0,0,0", Rotation="-5,0,0", Scale="1,1,1") {
    rotationMode=BMM_Additive
    rotationSpace=BCS_BoneSpace
    translationMode=BMM_Additive
    translationSpace=BCS_BoneSpace
    alphaBoolBlend {blendOption=Linear}
    boneToModify {boneName=dyn_hair_braid_MSTR}
    rotation {pitch=-5}
   }
   n26 = Control Rig!(Source=Component To Local(ComponentPose=@n25), bAlphaBoolEnabled:bool=Get hair() -> (hair:bool)) -> (Pose) {
    alphaInputType=Bool
    alphaBoolBlend {blendInTime=0.1, blendOption=Linear, blendOutTime=0.1}
    controlRigAssetReference {blueprintRigClass=&/F_MED_Galactic_FaceAcc_CtrlRig.F_MED_Galactic_FaceAcc_CtrlRig_C}
   }
   Save cached pose 'prerigid'(Pose=@n26) {cacheName=prerigid}
  ]}
 EventGraph {nodes [
   Event BlueprintUpdateAnimation()
   > Set head delta(head delta:double=GetAlphaFromReferenceTransformDelta(
     SkeletalMeshComponent=Get Skel_Mesh()
     BoneName=head
     BaseName=spine_05
     DeltaComponentType=QuaternionTwist
     InputRange=(X=20,Y=46)
     TwistAxis="0,0,1"
    )) -> (Output_Get:double)
   > Set head delta_fwd(head delta_fwd:double=GetAlphaFromReferenceTransformDelta(
     SkeletalMeshComponent=Get Skel_Mesh()
     BoneName=head
     BaseName=spine_05
     DeltaComponentType=QuaternionTwist
     InputRange=(X=0,Y=-20)
     TwistAxis="0,0,1"
    )) -> (Output_Get:double)
   > Set Arm Rot(Arm Rot:double=GetAlphaFromReferenceTransformDelta(SkeletalMeshComponent=Get Skel_Mesh(), BoneName=upperarm_r, BaseName=clavicle_r, InputRange=(X=55,Y=100), TwistAxis="0,0,1")) -> (Output_Get:double)
   > n1 = Cast To FortPlayerAnimInstance(Object=GetAnimInstance(self=GetOwnerSkeletalMeshComponent()))
   > Set weapon(weapon:String=Enum to String!(Enumerator:EFortWeaponCoreAnimation Enum=Get WeaponCoreAnim(self=@n1)) -> (ReturnValue:String)) -> (Output_Get:String)
   > n2 = Branch(Condition="Equal Exactly (String)"(A=Get weapon() -> (weapon:String), B=Pistol))
   > Set hair(hair:bool=true) -> (Output_Get:bool)
   Set hair(execute=@n2.else, hair=false)
   Event BlueprintInitializeAnimation()
   n3 = GetOwnerSkeletalMeshComponent()
   > Set Skel_Mesh(Skel_Mesh=@n3)
   > Set Body_Mesh(Body_Mesh=@n3)
   > n4 = Cast To PlayerPawn_Athena(Object=TryGetPawnOwner())
   > n5 = Set Pawn Owner(Pawn Owner:Player Pawn Athena Object Reference=@n4) -> (Output_Get:Player Pawn Athena Object Reference)
   > Set Head_Mesh(Head_Mesh:Skeletal Mesh Component Object Reference=GetSkeletalMeshForPartType(self=@n5, PartType=Head)) -> (Output_Get:Skeletal Mesh Component Object Reference)
  ]}
]
variables [
 Arm Rot {type=number}
 Body_Mesh {type=SkeletalMeshComponent}
 hair {type=boolean}
 head delta {type=number}
 head delta_fwd {type=number}
 Head_Mesh {type=SkeletalMeshComponent}
 Pawn Owner {type=PlayerPawn_Athena_C}
 Skel_Mesh {type=SkeletalMeshComponent}
 weapon {type=string, default=""}
]
